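# Library material function for VFX that converts an RGB input to an HSV vector, intended to match Niagara's conversion.
# Channel max gives value, and max minus min over max gives saturation.
# Nested If nodes find the dominant channel and select a custom HLSL hue formula, normalized from degrees to 0–1.
# Zero range forces hue to zero, while the similar saturation check leaves its equal branch unconnected.
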MaterialFunction /Game/Effects/Niagara/MasterMaterials/Functions/MF_FX_RGBtoHSV
graph {
 expressions {
  MaterialExpressionConstant_0 {class=MaterialExpressionConstant}
  MaterialExpressionConstant_6 {class=MaterialExpressionConstant}
  MaterialExpressionConstant_7 {class=MaterialExpressionConstant}
  MaterialExpressionConvert_1 {
   class=MaterialExpressionConvert
   inputs [None {from=MaterialExpressionFunctionInput_0}]
   properties {
    nodeName=Break Float3
    convertInputs [{
      type=Vector3
      expressionInput {
       expression=&self:MaterialExpressionFunctionInput_0
       inputName=None
       mask=0
       maskA=0
       maskB=0
       maskG=0
       maskR=0
       outputIndex=0
      }
     }]
    convertOutputs [{type=Scalar}, {type=Scalar}, {type=Scalar}]
   }
  }
  MaterialExpressionConvert_3 {
   class=MaterialExpressionConvert
   inputs [
    None {from=MaterialExpressionDivide_0}
    None {from=MaterialExpressionIf_5}
    None {from=MaterialExpressionReroute_10}
   ]
   properties {
    nodeName=Make Vector3
    convertInputs [
     {
      type=Scalar
      expressionInput {
       expression=&self:MaterialExpressionDivide_0
       inputName=None
       mask=0
       maskA=0
       maskB=0
       maskG=0
       maskR=0
       outputIndex=0
      }
     }
     {
      type=Scalar
      expressionInput {
       expression=&self:MaterialExpressionIf_5
       inputName=None
       mask=0
       maskA=0
       maskB=0
       maskG=0
       maskR=0
       outputIndex=0
      }
     }
     {
      type=Scalar
      expressionInput {
       expression=&self:MaterialExpressionReroute_10
       inputName=None
       mask=0
       maskA=0
       maskB=0
       maskG=0
       maskR=0
       outputIndex=0
      }
     }
    ]
    convertOutputs [{type=Vector3}]
   }
  }
  MaterialExpressionCustom_0 {
   class=MaterialExpressionCustom
   inputs [
    RGBRange {from=MaterialExpressionReroute_23, output=""}
    G {from=MaterialExpressionReroute_12, output=""}
    B {from=MaterialExpressionReroute_13, output=""}
   ]
   properties {
    code="Hue = fmod((((G - B) / RGBRange) * 60.0f) + 360.0f, 360.0f);\r\nreturn Hue;"
    outputType=CMOT_Float1
    showCode=true
    inputs [
     {
      inputName=RGBRange
      input {
       expression=&self:MaterialExpressionReroute_23
       inputName=None
       mask=0
       maskA=0
       maskB=0
       maskG=0
       maskR=0
       outputIndex=0
      }
     }
     {
      inputName=G
      input {
       expression=&self:MaterialExpressionReroute_12
       inputName=None
       mask=0
       maskA=0
       maskB=0
       maskG=0
       maskR=0
       outputIndex=0
      }
     }
     {
      inputName=B
      input {
       expression=&self:MaterialExpressionReroute_13
       inputName=None
       mask=0
       maskA=0
       maskB=0
       maskG=0
       maskR=0
       outputIndex=0
      }
     }
    ]
   }
  }
  MaterialExpressionCustom_1 {
   class=MaterialExpressionCustom
   inputs [
    RGBRange {from=MaterialExpressionReroute_23, output=""}
    R {from=MaterialExpressionReroute_11, output=""}
    B {from=MaterialExpressionReroute_13, output=""}
   ]
   properties {
    code="Hue = (((B - R) / RGBRange) * 60.0f) + 120.0f;\r\nreturn Hue;"
    outputType=CMOT_Float1
    showCode=true
    inputs [
     {
      inputName=RGBRange
      input {
       expression=&self:MaterialExpressionReroute_23
       inputName=None
       mask=0
       maskA=0
       maskB=0
       maskG=0
       maskR=0
       outputIndex=0
      }
     }
     {
      inputName=R
      input {
       expression=&self:MaterialExpressionReroute_11
       inputName=None
       mask=0
       maskA=0
       maskB=0
       maskG=0
       maskR=0
       outputIndex=0
      }
     }
     {
      inputName=B
      input {
       expression=&self:MaterialExpressionReroute_13
       inputName=None
       mask=0
       maskA=0
       maskB=0
       maskG=0
       maskR=0
       outputIndex=0
      }
     }
    ]
   }
  }
  MaterialExpressionCustom_2 {
   class=MaterialExpressionCustom
   inputs [
    RGBRange {from=MaterialExpressionReroute_23, output=""}
    R {from=MaterialExpressionReroute_11, output=""}
    G {from=MaterialExpressionReroute_12, output=""}
   ]
   properties {
    code="Hue = (((R - G) / RGBRange) * 60.0f) + 240.0f;\r\nreturn Hue;\r\n"
    outputType=CMOT_Float1
    showCode=true
    inputs [
     {
      inputName=RGBRange
      input {
       expression=&self:MaterialExpressionReroute_23
       inputName=None
       mask=0
       maskA=0
       maskB=0
       maskG=0
       maskR=0
       outputIndex=0
      }
     }
     {
      inputName=R
      input {
       expression=&self:MaterialExpressionReroute_11
       inputName=None
       mask=0
       maskA=0
       maskB=0
       maskG=0
       maskR=0
       outputIndex=0
      }
     }
     {
      inputName=G
      input {
       expression=&self:MaterialExpressionReroute_12
       inputName=None
       mask=0
       maskA=0
       maskB=0
       maskG=0
       maskR=0
       outputIndex=0
      }
     }
    ]
   }
  }
  MaterialExpressionDivide_0 {
   class=MaterialExpressionDivide
   inputs [
    A {from=MaterialExpressionIf_4}
    B {from=""}
   ]
   properties {constB=360, desc=HSV.Hue}
  }
  MaterialExpressionDivide_1 {
   class=MaterialExpressionDivide
   inputs [
    A {from=MaterialExpressionSubtract_1}
    B {from=MaterialExpressionReroute_7}
   ]
   properties {desc=Saturation}
  }
  MaterialExpressionFunctionInput_0 {class=MaterialExpressionFunctionInput, inputs [Preview {from="", output=""}], properties {inputName=RGB}}
  MaterialExpressionFunctionOutput_0 {class=MaterialExpressionFunctionOutput, inputs [None {from=MaterialExpressionConvert_3, output=""}]}
  MaterialExpressionIf_1 {
   class=MaterialExpressionIf
   inputs [
    A {from=MaterialExpressionReroute_16, output=""}
    B {from=MaterialExpressionReroute_15, output=""}
    "A > B" {from=MaterialExpressionReroute_14, output=""}
    "A == B" {from=MaterialExpressionCustom_0, output=""}
    "A < B" {from=MaterialExpressionReroute_14, output=""}
   ]
   properties {desc=Test Red}
  }
  MaterialExpressionIf_2 {
   class=MaterialExpressionIf
   inputs [
    A {from=MaterialExpressionReroute_19, output=""}
    B {from=MaterialExpressionReroute_18, output=""}
    "A > B" {from=MaterialExpressionReroute_24, output=""}
    "A == B" {from=MaterialExpressionCustom_1, output=""}
    "A < B" {from=MaterialExpressionReroute_24, output=""}
   ]
   properties {desc=Test Green}
  }
  MaterialExpressionIf_3 {
   class=MaterialExpressionIf
   inputs [
    A {from=MaterialExpressionReroute_26, output=""}
    B {from=MaterialExpressionReroute_25, output=""}
    "A > B" {from=MaterialExpressionConstant_0, output=""}
    "A == B" {from=MaterialExpressionCustom_2, output=""}
    "A < B" {from=MaterialExpressionConstant_0, output=""}
   ]
   properties {desc=Test Blue}
  }
  MaterialExpressionIf_4 {
   class=MaterialExpressionIf
   inputs [
    A {from=MaterialExpressionReroute_23, output=""}
    B {from="", output=""}
    "A > B" {from=MaterialExpressionIf_1, output=""}
    "A == B" {from=MaterialExpressionConstant_6, output=""}
    "A < B" {from=MaterialExpressionIf_1, output=""}
   ]
   properties {desc="HSV.Hue: if 0: Hue = 0"}
  }
  MaterialExpressionIf_5 {
   class=MaterialExpressionIf
   inputs [
    A {from=MaterialExpressionReroute_23, output=""}
    B {from=MaterialExpressionConstant_7, output=""}
    "A > B" {from=MaterialExpressionReroute_28, output=""}
    "A == B" {from="", output=""}
    "A < B" {from=MaterialExpressionReroute_28, output=""}
   ]
   properties {desc="HSV.Saturation: if 0: Sat = 0"}
  }
  MaterialExpressionMax_0 {
   class=MaterialExpressionMax
   inputs [
    A {from=MaterialExpressionReroute_2}
    B {from=MaterialExpressionReroute_4}
   ]
  }
  MaterialExpressionMax_1 {
   class=MaterialExpressionMax
   inputs [
    A {from=MaterialExpressionReroute_5}
    B {from=MaterialExpressionMax_0}
   ]
  }
  MaterialExpressionMin_0 {
   class=MaterialExpressionMin
   inputs [
    A {from=MaterialExpressionReroute_2, output=""}
    B {from=MaterialExpressionReroute_4, output=""}
   ]
   properties {}
  }
  MaterialExpressionMin_1 {
   class=MaterialExpressionMin
   inputs [
    A {from=MaterialExpressionMin_0, output=""}
    B {from=MaterialExpressionReroute_5, output=""}
   ]
   properties {}
  }
  MaterialExpressionReroute_10 {class=MaterialExpressionReroute, inputs [None {from=MaterialExpressionReroute_9}], properties {desc=HSV.Value}}
  MaterialExpressionReroute_11 {class=MaterialExpressionReroute, inputs [None {from=MaterialExpressionConvert_1}], properties {desc=R}}
  MaterialExpressionReroute_12 {class=MaterialExpressionReroute, inputs [None {from=MaterialExpressionConvert_1}], properties {desc="G\r\n"}}
  MaterialExpressionReroute_13 {class=MaterialExpressionReroute, inputs [None {from=MaterialExpressionConvert_1}], properties {desc="B\r\n"}}
  MaterialExpressionReroute_14 {class=MaterialExpressionReroute, inputs [None {from=MaterialExpressionIf_2}]}
  MaterialExpressionReroute_15 {class=MaterialExpressionReroute, inputs [None {from=MaterialExpressionReroute_17}]}
  MaterialExpressionReroute_16 {class=MaterialExpressionReroute, inputs [None {from=MaterialExpressionReroute_9}]}
  MaterialExpressionReroute_17 {class=MaterialExpressionReroute, inputs [None {from=MaterialExpressionReroute_11}], properties {desc=R}}
  MaterialExpressionReroute_18 {class=MaterialExpressionReroute, inputs [None {from=MaterialExpressionReroute_20}]}
  MaterialExpressionReroute_19 {class=MaterialExpressionReroute, inputs [None {from=MaterialExpressionReroute_9}]}
  MaterialExpressionReroute_2 {class=MaterialExpressionReroute, inputs [None {from=MaterialExpressionConvert_1}]}
  MaterialExpressionReroute_20 {class=MaterialExpressionReroute, inputs [None {from=MaterialExpressionReroute_12}], properties {desc=G}}
  MaterialExpressionReroute_23 {class=MaterialExpressionReroute, inputs [None {from=MaterialExpressionSubtract_1}], properties {desc=Longest Comp}}
  MaterialExpressionReroute_24 {class=MaterialExpressionReroute, inputs [None {from=MaterialExpressionIf_3}]}
  MaterialExpressionReroute_25 {class=MaterialExpressionReroute, inputs [None {from=MaterialExpressionReroute_27}]}
  MaterialExpressionReroute_26 {class=MaterialExpressionReroute, inputs [None {from=MaterialExpressionReroute_9}]}
  MaterialExpressionReroute_27 {class=MaterialExpressionReroute, inputs [None {from=MaterialExpressionReroute_13}], properties {desc=B}}
  MaterialExpressionReroute_28 {class=MaterialExpressionReroute, inputs [None {from=MaterialExpressionDivide_1}]}
  MaterialExpressionReroute_4 {class=MaterialExpressionReroute, inputs [None {from=MaterialExpressionConvert_1}]}
  MaterialExpressionReroute_5 {class=MaterialExpressionReroute, inputs [None {from=MaterialExpressionConvert_1}]}
  MaterialExpressionReroute_7 {class=MaterialExpressionReroute, inputs [None {from=MaterialExpressionMax_1}], properties {desc=MAX}}
  MaterialExpressionReroute_8 {class=MaterialExpressionReroute, inputs [None {from=MaterialExpressionMin_1}], properties {desc=MIN}}
  MaterialExpressionReroute_9 {class=MaterialExpressionReroute, inputs [None {from=MaterialExpressionReroute_7}], properties {desc=MAX}}
  MaterialExpressionSubtract_1 {
   class=MaterialExpressionSubtract
   inputs [
    A {from=MaterialExpressionReroute_7}
    B {from=MaterialExpressionReroute_8}
   ]
   properties {desc=Longest Comp}
  }
 }
 properties {bExposeToLibrary=true, description="RGB to HSV, matching Niagara", userExposedCaption=RGB to HSV, libraryCategoriesText [Misc,VFX]}
}
