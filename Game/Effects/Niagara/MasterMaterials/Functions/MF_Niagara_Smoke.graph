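# MF_Niagara_Smoke is a material function in the Niagara master materials that gives smoke effects named access to per-particle dynamic parameters.
# A single Dynamic Parameter node names its first two channels Erode and Shading and passes each straight to a function output of the same name.
# The other two channels keep generic names and are not output, and the function does no math of its own.
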MaterialFunction /Game/Effects/Niagara/MasterMaterials/Functions/MF_Niagara_Smoke
graph {nodes [
  n1 = DynamicParameter() {paramNames [Erode,Shading,Param3,Param4]}
  FunctionOutput(None=@n1.Erode) {outputName=Erode}
  FunctionOutput(None=@n1.Shading) {outputName=Shading, sortPriority=1}
 ]}
