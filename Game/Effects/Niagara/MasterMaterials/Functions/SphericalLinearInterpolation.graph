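# Material function that spherically interpolates (slerps) between quaternions A and B by an Alpha bias.
# It normalizes both inputs, derives theta from the arccosine of their clamped dot product, blends them with sine-ratio weights and renormalizes.
# A fallback branch, using an error margin for near-parallel or opposite inputs, feeds nothing.
# The output takes the raw slerp, without guarding zero sinTheta or flipping sign for shortest path.
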MaterialFunction /Game/Effects/Niagara/MasterMaterials/Functions/SphericalLinearInterpolation
graph {nodes [
  n1 = NamedRerouteDeclaration(None=Normalize(VectorInput=FunctionInput() {description=The inital quaternion you are transforming from., inputName=A, inputType=FunctionInput_Vector4})) {name=startQuat, nodeColor=<rgba 0 0.7422 1 1>}
  n2 = FunctionInput() {description=Interpolation bias between the two quaternions., inputName=Alpha, inputType=FunctionInput_Scalar, sortPriority=2}
  n3 = NamedRerouteDeclaration(None=@n2) {name=Bias, nodeColor=<rgba 1 0.4688 0 1>}
  n4 = OneMinus(None=@n3)
  n5 = NamedRerouteDeclaration(None=Normalize(VectorInput=FunctionInput() {description=The final quaternion you are transforming to., inputName=B, inputType=FunctionInput_Vector4, sortPriority=1})) {name=endQuat, nodeColor=<rgba 1 0 0.5625 1>}
  n6 = DotProduct(A=@n1, B=@n5)
  n7 = Arccosine!(None=Clamp(None=NamedRerouteDeclaration(None=@n6) {name=dotP, nodeColor=<rgba 0.2656 1 0 1>}) {minDefault=-1}) {desc=theta}
  n8 = Sine(None=@n7) {desc=sinTheta}
  n9 = Divide(A=Sine(None=Multiply(A=@n7, B=@n4)), B=@n8)
  n10 = Divide(A=Sine(None=Multiply(A=@n7, B=@n3)), B=@n8)
  n11 = Normalize(VectorInput=Add(A=Multiply(A=@n1, B=@n9), B=Multiply(A=@n5, B=@n10)))
  FunctionOutput(None=@n11) {description=Interpolated quaternion., outputName=None}
  n12 = NamedRerouteUsage()
  n13 = Constant() {r=0.5}
  n14 = NamedRerouteUsage()
  n15 = FunctionInput() {
   description=Error Margin
   inputName=DotP Error Margin
   inputType=FunctionInput_Scalar
   sortPriority=3
   previewValue=<0.9999 0.9999 0.9999 0.9999>
  }
  n16 = NamedRerouteUsage()
  n17 = Multiply(A=@n15) {constB=-1}
  LinearInterpolate(
   A=@n11
   B=LinearInterpolate(
    A=NamedRerouteUsage()
    B=NamedRerouteUsage()
    Alpha=MaterialFunctionCall!(A=@n12, B=@n13) {materialFunction=/FortUILibrary/MaterialFunctions/Basics/MF_UI_BooleanOperations}.IsLessOrEqual
   )
   Alpha=MaterialFunctionCall(
    A=MaterialFunctionCall(A=@n14, B=@n15) {materialFunction=&MF_UI_BooleanOperations}.IsGreater
    B=MaterialFunctionCall(A=@n16, B=@n17) {materialFunction=&MF_UI_BooleanOperations}.IsLess
   ) {materialFunction=&MF_UI_BooleanOperations}.Or
  )
 ]}
