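# Two-sided lit master material for Niagara mesh particles that renders purple cube growths that grow, jiggle and glow.
# Dynamic parameters and vertex colors drive a staggered grow-from-point offset with sine jiggle, plus a glow mask from a sphere or vertex red.
# Parallax, panning cloud noise and cube noise shape glow, base color and edge-faded normal detail.
# Fake view-reflection specular, disabled on low quality, adds highlights; deferred shading remaps emissive to HDR.
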Material /Game/Effects/Niagara/MasterMaterials/Mesh/M_CubeGrowth_Mesh
graph {
 properties {bAutomaticallySetUsageInEditor=false, bUsedWithNiagaraMeshParticles=true, twoSided=true, usedShadingModels=Default Lit}
 nodes [
  n1 = ReflectionVectorWS()
  n2 = CameraVectorWS()
  n3 = Saturate(None=NamedRerouteUsage())
  n4 = Power(
   Base=Saturate(None=DotProduct(A=@n1, B=@n2))
   Exp=LinearInterpolate(A=ScalarParameter() {defaultValue=40, parameterName=FakeSpecPower}, B=Constant() {r=10}, Alpha=@n3)
  )
  n5 = LinearInterpolate(Alpha=@n3) {constA=1, constB=0.4}
  n6 = QualitySwitch(
   Default=Multiply(A=Multiply(A=@n4, B=@n5), B=VectorParameter() {parameterName=FakeSpecularColor, defaultValue=<rgba 0.0826 0.0429 0.1 1>}.RGB)
   Low=Constant()
  )
  n7 = Add(
   A=ParticleRandom()
   B=Multiply(A=TextureCoordinate() {coordinateIndex=1}, B=ScalarParameter() {defaultValue=6, parameterName=CubeNoiseTiling})
  )
  n8 = Multiply(A=@n7) {constB=2}
  n9 = TextureSample(UVs=@n8) {samplerType=SAMPLERTYPE_LinearColor, texture=/Game/Athena/Environments/Island/CubeGrowth/T_CubeNoise01_H}
  n10 = LinearInterpolate(Alpha=NamedRerouteUsage()) {constA=0.2}
  n11 = MaterialFunctionCall(Transition Phase=Power(Base=Saturate(None=@n10)), Height Texture=NamedRerouteUsage()) {materialFunction=&HeightLerp}
  Add(
   A=@n6
   B=LinearInterpolate(
    A=VectorParameter() {parameterName=BaseColor, defaultValue=<rgba 0.0704 0.0194 0.0999 1>}.RGB
    B=VectorParameter() {parameterName=EdgeColor, defaultValue=<rgba 0.013 0.0044 0.0307 1>}.RGB
    Alpha=Multiply(A=@n9.G, B=@n11.Lerp Alpha No Contrast)
   )
  )
  n12 = NamedRerouteUsage()
  n13 = WorldPosition()
  n14 = DynamicParameter() {parameterIndex=1, defaultValue=<rgba 1 0 1 1>, paramNames [Grow,JiggleScale,GlowGain,Param4]}
  n15 = VertexColor()
  n16 = Subtract(A=Multiply(A=@n14.Grow) {constB=2}, B=@n15.G)
  n17 = Time()
  n18 = Multiply(A=Add(A=Add(A=@n15.B, B=@n17), B=ParticleRandom()), B=ScalarParameter() {defaultValue=5, parameterName=JiggleSpeed})
  n19 = Sine(None=@n18) {period=0.83}
  n20 = Sine(None=@n18)
  n21 = Sine(None=@n18) {period=1.042}
  Add(
   A=LinearInterpolate(A=Subtract(A=@n12, B=@n13.XYZ), Alpha=Saturate(None=@n16)) {constB=0}
   B=Multiply(
    A=@n14.JiggleScale
    B=MaterialFunctionCall(R=@n19, G=@n20, B=@n21) {materialFunction=&AppendMany}.RGB
   )
  )
  Constant() {r=0.5}
  Constant() {r=0.2}
  Constant() {r=1}
  n22 = Saturate(None=NamedRerouteUsage())
  n23 = LinearInterpolate(
   A=VectorParameter() {parameterName=E1, defaultValue=<rgba 0.011 0 0.0137 1>}.RGB
   B=VectorParameter() {parameterName=E2, defaultValue=<rgba 0.7084 0 1 1>}.RGB
   Alpha=@n22
  )
  n24 = Power(Base=@n22) {constExponent=0.7}
  n25 = Add(A=Multiply(A=Multiply(A=@n23, B=@n24), B=ParticleColor().RGB), B=@n6)
  n26 = MaterialFunctionCall(LDR Input=Saturate(None=@n25), HDR Tint=@n25, HDR Intensity Multiplier=ScalarParameter() {defaultValue=4, parameterName=HDRIntensity}) {materialFunction=&DeriveHDRfromLDR}
  MaterialFunctionCall(Emissive=Multiply(A=ShadingPathSwitch(Default=@n25, Deferred=@n26.RemappedRange), B=@n14.GlowGain)) {materialFunction=&MF_FX_EmissiveAdjustments}
  n27 = TextureCoordinate() {coordinateIndex=2}
  n28 = ScalarParameter() {defaultValue=0.6, parameterName=CloudNoiseScale}
  n29 = Rotator(Coordinate=BumpOffset(
    Coordinate=Multiply(A=@n27, B=@n28)
    Height=ScalarParameter() {defaultValue=0.58, parameterName=CloudParallax}
    HeightRatioInput=Constant() {r=-0.5}
   )) {centerX=0.25, centerY=0.333, speed=0.0075}
  n30 = ScalarParameter() {defaultValue=0.004, parameterName=CloudSpeed}
  n31 = Power(
   Base=TextureSample!(UVs=Panner(Coordinate=@n29, Speed=@n30), Tex, Level, Apply View MipBias) {mipValueMode=TMVM_MipLevel, samplerType=SAMPLERTYPE_Masks, texture=/Game/Global/Textures/Default/Noise/T_CubeNoise}.RGB
   Exp=ScalarParameter() {defaultValue=3, parameterName=CloudPower}
  )
  NamedRerouteDeclaration(None=ComponentMask(None=@n31) {r=true}) {name=Cloud Noise Red Channel, nodeColor=<rgba 1 0 0.0174 1>}
  NamedRerouteDeclaration(None=ComponentMask(None=@n31) {g=true}) {name=Cloud Noise Green Channel, nodeColor=<rgba 0.0093 1 0 1>}
  n32 = DynamicParameter() {defaultValue=<rgba 0 0 -150 200>, paramNames [GlowDir X, GlowDir Y, GlowDir Z, GlowRadius]}
  n33 = StaticSwitchParameter(
   True=SphereMask(
    A=WorldPosition().XYZ
    B=NamedRerouteDeclaration(None=@n32.RGB) {name=GrowOriginPosition, nodeColor=<rgba 0 1 0.4609 1>}
    Radius=@n32.GlowRadius
   ) {hardnessPercent=0}
   False=Add(A=VertexColor().R, B=@n32.GlowDir X)
  ) {defaultValue=true, parameterName=UseDynParamsForGlowMask}
  n34 = Power(Base=@n33)
  n35 = LinearInterpolate(Alpha=@n34) {constB=0.5}
  n36 = NamedRerouteUsage()
  n37 = Multiply(A=NamedRerouteUsage(), B=@n34)
  n38 = TextureSample!(UVs=@n7, Tex, Apply View MipBias) {samplerType=SAMPLERTYPE_Masks, texture=&T_CubeNoise}
  n39 = Multiply(A=Multiply(A=Power(Base=@n33) {constExponent=4}, B=OneMinus(None=@n38.R))) {constB=0.6}
  NamedRerouteDeclaration(None=Multiply(A=LinearInterpolate(B=@n35, Alpha=@n36) {constA=1, constB=0.2}, B=Add(A=@n37, B=@n39))) {name=GlowMask, nodeColor=<rgba 0.2188 0 1 1>}
  n40 = TextureCoordinate() {coordinateIndex=2}
  n41 = Constant2Vector() {g=0.5, r=0.5}
  n42 = Power(Base=Abs(None=Subtract(A=@n40, B=@n41)), Exp=ScalarParameter() {defaultValue=2, parameterName=EdgeMaskExp})
  n43 = ScalarParameter() {defaultValue=2, parameterName=EdgeMaskMulti}
  n44 = MaterialFunctionCall(Float2=Saturate(None=Multiply(A=@n42, B=@n43))) {materialFunction=&BreakOutFloat2Components}
  NamedRerouteDeclaration(None=Max(A=@n44.R, B=@n44.R)) {name=Edge Mask, nodeColor=<rgba 0 0.9062 1 1>}
  n45 = ScalarParameter() {defaultValue=0.1, parameterName=R1}
  n46 = ScalarParameter() {defaultValue=0.6, parameterName=R2}
  NamedRerouteDeclaration(None=Add(A=OneMinus(None=@n9.G), B=LinearInterpolate(A=@n45, B=@n46, Alpha=@n38.R)) {constB=0}) {name=Noise, nodeColor=<rgba 0.7344 1 0 1>}
  NamedRerouteUsage()
  n47 = Multiply(
   A=Constant3Vector() {constant=<rgba 1 1 0 0>}
   B=TextureSample(UVs=@n8) {samplerType=SAMPLERTYPE_Normal, texture=/Game/Athena/Environments/Island/CubeGrowth/T_CubeNoise01}.RGB
  )
  n48 = TextureCoordinate() {coordinateIndex=2}
  n49 = Abs(None=Multiply(A=Subtract(A=@n48) {constB=0.5}) {constB=2})
  n50 = ComponentMask(None=@n49) {r=true}
  n51 = ComponentMask(None=@n49) {g=true}
  n52 = Power(Base=Saturate(None=Max(A=@n50, B=@n51))) {constExponent=5}
  n53 = LinearInterpolate(
   A=ScalarParameter() {defaultValue=0.3, parameterName=CubeNoiseNormalMax}
   B=ScalarParameter() {defaultValue=0.05, parameterName=CubeNoiseNormalMin}
   Alpha=Max(A=NamedRerouteDeclaration(None=@n52) {name=EdgeMask, nodeColor=<rgba 0 1 0.2266 1>}, B=VertexColor().B)
  )
  Reroute(None=Add(
    A=TextureSample() {samplerType=SAMPLERTYPE_Normal, texture=/Game/Athena/Environments/Island/CubeGrowth/T_CubeGrowth_normals}.RGB
    B=Multiply(A=@n47, B=@n53)
   ))
  ScalarParameter() {defaultValue=1, parameterName=Glow Gain}
 ]
}
