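# Unlit translucent Niagara master material for mesh particles and sprites, producing a Fresnel glow that defaults to inverted, fading toward silhouette edges.
# Fresnel and optional sine UV mask drive opacity and, with optional noise, a two-color emissive blend tinted by particle color.
# Vertices get camera offset plus UV-masked normal push, adjustable through a dynamic parameter, and optional noise displacement.
# Quality switches limit depth fade and vertex deformation to High and Epic.
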Material /Game/Effects/Niagara/MasterMaterials/Mesh/M_InvertedGlow
graph {
 properties {
  bAutomaticallySetUsageInEditor=false
  bUsedWithNiagaraMeshParticles=true
  bUsedWithNiagaraSprites=true
  blendMode=BLEND_Translucent
  shadingModel=MSM_Unlit
  usedShadingModels=Unlit
 }
 nodes [
  n1 = ParticleColor()
  n2 = VectorParameter() {group=2. Color, parameterName=Color 2, sortPriority=2, defaultValue=<rgba 0.5395 0 0 1>}
  n3 = VectorParameter() {group=2. Color, parameterName=Color, sortPriority=1, defaultValue=<rgba 5 0.4995 0 1>}
  n4 = TextureCoordinate()
  n5 = VectorParameter() {
   group=6. Noise
   parameterName=Noise UV Control
   sortPriority=2
   channelNames {a=Offset V, b=Offset U, g=Tiling V, r=Tiling U}
   defaultValue=<rgba 1 1 0 0>
  }
  n6 = AppendVector(A=@n5.Tiling U, B=@n5.Tiling U)
  n7 = AppendVector(A=@n5.Offset U, B=@n5.Offset U)
  n8 = Time()
  n9 = TextureSampleParameter2D(UVs=Add(A=Multiply(A=@n4, B=@n6), B=Multiply(A=@n7, B=@n8))) {group=6. Noise, parameterName=Noise Tex, sortPriority=1, texture=/Game/Global/Textures/Default/Noise/T_StringyEnergy_Noise}
  n10 = ScalarParameter() {group=6. Noise, parameterName=Noise Weight, sortPriority=0}
  n11 = Fresnel(
   ExponentIn=ScalarParameter() {defaultValue=0.79, group=1. Fresnel, parameterName=ExponentIn, sortPriority=0}
   BaseReflectFractionIn=ScalarParameter() {defaultValue=0.106, group=1. Fresnel, parameterName=BaseReflectFractionIn, sortPriority=1}
  )
  n12 = LinearInterpolate(A=OneMinus(None=@n11), B=@n11, Alpha=ScalarParameter() {group=1. Fresnel, parameterName=Invert Fresnel})
  n13 = ScalarParameter() {defaultValue=3, group=1. Fresnel, parameterName=Fresnel Power, sortPriority=2}
  n14 = TextureCoordinate()
  n15 = Power(
   Base=Sine(None=ComponentMask(None=@n14) {r=true}) {period=2}
   Exp=ScalarParameter() {defaultValue=1, group=5. Mask, parameterName=Mask V Power, sortPriority=2}
  )
  n16 = TextureCoordinate()
  n17 = Power(
   Base=Sine(None=ComponentMask(None=@n16) {g=true}) {period=2}
   Exp=ScalarParameter() {defaultValue=1, group=5. Mask, parameterName=Mask H Power, sortPriority=1}
  ) {constExponent=1}
  n18 = Multiply(
   A=Saturate(None=Power(Base=@n12, Exp=@n13))
   B=LinearInterpolate(B=Multiply(A=@n15, B=@n17), Alpha=ScalarParameter() {group=5. Mask, parameterName=Use Simple Mask, sortPriority=0}) {constA=1}
  )
  n19 = Add(A=ScalarParameter() {defaultValue=-0.2, group=2. Color, parameterName=Color Offset, sortPriority=3}, B=@n18)
  n20 = Multiply(
   A=LinearInterpolate(B=@n9.R, Alpha=Saturate(None=@n10)) {constA=1}
   B=Power(Base=Saturate(None=@n19), Exp=ScalarParameter() {defaultValue=1, group=2. Color, parameterName=Color Power, sortPriority=5}) {constExponent=1}
  )
  MaterialFunctionCall(Emissive=Multiply(A=@n1.RGB, B=LinearInterpolate(A=@n2.RGB, B=@n3.RGB, Alpha=@n20))) {materialFunction=&MF_FX_EmissiveAdjustments}
  n21 = DepthFade(Opacity=@n18, FadeDistance=ScalarParameter() {defaultValue=2.5, group=3. Modifiers, parameterName=DepthFade, sortPriority=1})
  Multiply(
   A=@n1.A
   B=Multiply(
    A=ScalarParameter() {defaultValue=1, group=3. Modifiers, parameterName=Opacity, sortPriority=0}
    B=QualitySwitch(Default=@n18, Low=@n18, High=@n21, Medium=@n18, Epic=@n21)
   )
  )
  n22 = MaterialFunctionCall(Offset Amount=ScalarParameter() {defaultValue=-0.2, group=3. Modifiers, parameterName=CameraOffset}) {materialFunction=/Engine/Functions/Engine_MaterialFunctions02/WorldPositionOffset/CameraOffset}
  n23 = Constant() {r=1}
  n24 = VectorParameter() {
   group=4. WPO
   parameterName=WPO Noise UV Controls
   sortPriority=5
   channelNames {a=Panner V, b=Panner U, g=Tiling V, r=Tiling U}
   defaultValue=<rgba 1 1 0 0>
  }
  n25 = Multiply(A=TextureCoordinate(), B=AppendVector(A=@n24.Tiling U, B=@n24.Tiling U)) {constB=2}
  n26 = ParticleRandom()
  n27 = TextureSampleParameter2D(UVs=Panner(Coordinate=Add(A=@n25, B=@n26), Speed=AppendVector(A=@n24.Panner U, B=@n24.Panner U)) {speedX=0.2, speedY=0.15}) {
   group=4. WPO
   parameterName=WPO Noise
   samplerType=SAMPLERTYPE_LinearColor
   sortPriority=4
   texture=/Game/Effects/SnowSandInteraction/Textures/SSI_Perlin
  }
  n28 = Multiply(
   A=VertexNormalWS()
   B=LinearInterpolate(
    A=ScalarParameter() {defaultValue=-0.6, group=4. WPO, parameterName=WPO Noise Min, sortPriority=6}
    B=ScalarParameter() {defaultValue=0.6, group=4. WPO, parameterName=WPO Noise Max, sortPriority=7}
    Alpha=@n27.R
   ) {constA=-0.3, constB=0.3}
  )
  n29 = Multiply(
   A=QualitySwitch(Default=@n23, Low=@n23, High=@n28, Medium=@n23, Epic=@n28)
   B=Saturate(None=ScalarParameter() {group=4. WPO, parameterName=Use WPO Noise, sliderMax=1, sortPriority=3})
  )
  n30 = Multiply(
   A=VertexNormalWS()
   B=Add(
    A=ScalarParameter() {defaultValue=3, group=4. WPO, parameterName=WPO Push, sortPriority=0}
    B=DynamicParameter() {defaultValue=<rgba 0 1 1 1>, paramNames [WPO Push, Param2, Param3, Param4]}.WPO Push
   )
  )
  n31 = ComponentMask(None=TextureCoordinate()) {r=true}
  n32 = Add(A=Sine(None=Multiply(A=@n31) {constB=0.5}), B=ScalarParameter() {group=4. WPO, parameterName=Ignore UV For WPO, sortPriority=2})
  n33 = Add(A=@n22.Clamped Camera Offset, B=LinearInterpolate(B=Add(A=@n29, B=@n30), Alpha=Saturate(None=@n32)))
  QualitySwitch(Default=Constant(), Low=@n22.Clamped Camera Offset, High=@n33, Medium=@n22.Clamped Camera Offset, Epic=@n33)
 ]
}
