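# Unlit, two-sided translucent master material for Niagara mesh and ribbon trails.
# A panning distortion map warps base and optional opacity UVs; the base texture blends two colors into HDR emissive.
# Opacity combines texture, procedural ribbon rounding, end mask, particle alpha, depth and camera fades.
# World position offset adds SplineThicken, U-weighted noise and optional local offsets from dynamic parameters; a rotate-about-axis chain is left unconnected.
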Material /Game/Effects/Niagara/MasterMaterials/Mesh/M_MeshTrail
graph {
 properties {
  bUseMaterialAttributes=true
  bUsedWithNiagaraMeshParticles=true
  bUsedWithNiagaraRibbons=true
  blendMode=BLEND_Translucent
  pixelDepthOffsetMode=PDOM_Legacy
  shadingModel=MSM_Unlit
  twoSided=true
  usedShadingModels=Unlit
 }
 nodes [
  n1 = TextureCoordinate()
  n2 = AppendVector(
   A=ScalarParameter() {defaultValue=1, group="1.1) Base Texture UVs", parameterName="Tile U (Base)", sortPriority=1}
   B=ScalarParameter() {defaultValue=1, group="1.1) Base Texture UVs", parameterName="Tile V (Base)", sortPriority=2}
  )
  n3 = LinearInterpolate(
   A=Multiply(A=@n1, B=@n2)
   B=MaterialFunctionCall(UVs=@n1, Texture Scale=@n2) {materialFunction=&ScaleUVsByCenter}.UVs
   Alpha=ScalarParameter() {group="1.1) Base Texture UVs", parameterName="Scale UVs From Center (Base)"}
  )
  n4 = Multiply(
   A=AppendVector(
    A=ScalarParameter() {group="1.1) Base Texture UVs", parameterName="Pan U Speed (Base)", sortPriority=3}
    B=ScalarParameter() {group="1.1) Base Texture UVs", parameterName="Pan V Speed (Base)", sortPriority=4}
   )
   B=DynamicParameter() {defaultValue=<rgba 0 0 0 1>, paramNames [R Offset, G Offset, B Offset, Speed - All]}.Speed - All
  )
  n5 = TextureCoordinate()
  n6 = TextureCoordinate() {coordinateIndex=1}
  n7 = ScalarParameter() {group="0.1) Distortion Texture UVs", parameterName="UV Channel (Distort)", sliderMax=1}
  n8 = ScalarParameter() {defaultValue=1, group="0.1) Distortion Texture UVs", parameterName="Tile U (Distort)", sortPriority=1}
  n9 = ScalarParameter() {defaultValue=1, group="0.1) Distortion Texture UVs", parameterName="Tile V (Distort)", sortPriority=2}
  n10 = ScalarParameter() {defaultValue=1, group="0.1) Distortion Texture UVs", parameterName="Pan U Speed (Distort)", sortPriority=3}
  n11 = ScalarParameter() {group="0.1) Distortion Texture UVs", parameterName="Pan V Speed (Distort)", sortPriority=4}
  n12 = Panner(
   Coordinate=Multiply(A=LinearInterpolate(A=@n5, B=@n6, Alpha=@n7), B=AppendVector(A=@n8, B=@n9)) {desc=This might need a frac option for getting textures to tile. This scales the texture.}
   Speed=Multiply(
    A=AppendVector(A=@n10, B=@n11)
    B=DynamicParameter() {defaultValue=<rgba 0 0 0 1>, paramNames [R Offset, G Offset, B Offset, Speed - All]}.Speed - All
   )
  )
  n13 = TextureSampleParameter2D(UVs=@n12) {
   group="0.0) Distortion Texture"
   parameterName=DistortionMap
   samplerType=SAMPLERTYPE_Normal
   texture=&T_Weap_Distort_N_02
  }
  n14 = Multiply(
   A=ComponentMask(None=@n13.RGB) {g=true, r=true}
   B=AppendVector(
    A=ScalarParameter() {defaultValue=0.1, group="0.0) Distortion Texture", parameterName="U Strength (Distort)", sortPriority=1}
    B=ScalarParameter() {defaultValue=0.1, group="0.0) Distortion Texture", parameterName="V Strength (Distort)", sortPriority=2}
   )
  )
  n15 = MaterialFunctionCall!(In, Uvs=Add(A=Panner(Coordinate=@n3, Speed=@n4), B=@n14), Base Color Value) {desc=Read texture, materialFunction=/Game/Effects/Niagara/MasterMaterials/Functions/V2/BaseColor/MF_FX_BaseColorTexture_v02}
  n16 = VectorParameter() {group="1.0) Base Color Texture", parameterName=Low Color, sortPriority=3, defaultValue=<rgba 0 0 0 1>}
  n17 = VectorParameter() {group="1.0) Base Color Texture", parameterName=High Color, sortPriority=2, defaultValue=<rgba 1 1 1 1>}
  n18 = ScalarParameter() {defaultValue=1, group="1.2) Base Color Correction", parameterName="PostMult (Base Color)", sortPriority=2}
  n19 = MaterialFunctionCall(Color=Multiply(
    A=@n15.Base Color Value
    B=VectorParameter() {group="1.2) Base Color Correction", parameterName=Mask Texture Channels, defaultValue=<rgba 1 1 1 1>}.RGB
   )) {materialFunction=&SplitComponents}
  n20 = Add(
   A=Add(A=Add(A=@n19.R, B=@n19.R), B=@n19.B)
   B=ScalarParameter() {group="1.2) Base Color Correction", parameterName="Add (Base Color)", sortPriority=1}
  )
  n21 = Max(A=Power(
    Base=Multiply(A=@n18, B=@n20)
    Exp=ScalarParameter() {defaultValue=1, group="1.2) Base Color Correction", parameterName="Power (Base Color)", sortPriority=3}
   ) {constExponent=0.5}) {constB=0}
  n22 = Multiply(
   A=LinearInterpolate(A=@n16.RGB, B=@n17.RGB, Alpha=@n21)
   B=LinearInterpolate(B=ParticleColor().RGB, Alpha=ScalarParameter() {group="1.0) Base Color Texture", parameterName=Use Particle Color/Alpha, sortPriority=1}) {constA=1}
  )
  n23 = SetMaterialAttributes!(MaterialAttributes=@n15.MA BaseColor, Base Color=@n22) {attributeSetTypes [#1]}
  n24 = MaterialFunctionCall(
   LDR Input=Saturate(None=@n22)
   HDR Tint=VectorParameter() {group="2.0) Emissive", parameterName=Emissive HDR Tint, defaultValue=<rgba 1 1 1 0>}.RGB
   HDR Intensity Multiplier=ScalarParameter() {defaultValue=1, group="2.0) Emissive", parameterName=Emissive HDR Tint Multiplier, sortPriority=1}
   Desaturate Luminance=ScalarParameter() {defaultValue=1, group="2.0) Emissive", parameterName=HDR Luminance Desaturate, sortPriority=2}
  ) {desc=remap colors for emissive, materialFunction=&DeriveHDRfromLDR}
  n25 = TextureCoordinate()
  n26 = Multiply(A=Subtract(A=ComponentMask(None=@n25) {g=true}, B=Constant() {r=0.5}), B=Constant() {r=2})
  n27 = Saturate(None=Multiply(A=@n26, B=@n26))
  n28 = ComponentMask(None=@n25) {r=true}
  n29 = ScalarParameter() {
   defaultValue=0.5
   desc=Amount of rounding on the base of the ribbon
   group="6.1) Opacity"
   parameterName=Ribbon Rounding
   sliderMax=1
   sortPriority=11
  }
  n30 = OneMinus(None=Saturate(None=Divide(A=@n28, B=@n29)))
  n31 = ScalarParameter() {defaultValue=0.2, desc=Softness, group="6.1) Opacity", parameterName=Ribbon Rounding Softness, sliderMax=1, sortPriority=12}
  n32 = OneMinus(None=Abs(None=@n26))
  n33 = Divide(A=Subtract(A=OneMinus(None=@n27), B=Multiply(A=@n30, B=@n30)), B=LinearInterpolate(A=@n31, Alpha=Multiply(A=@n32, B=@n32)))
  n34 = MaterialFunctionCall(Vector 3=@n21) {materialFunction=/Engine/Functions/Engine_MaterialFunctions02/Utility/VectorLength}
  n35 = LinearInterpolate(
   A=TextureCoordinate()
   B=TextureCoordinate() {coordinateIndex=1}
   Alpha=ScalarParameter() {group="6.2) Opacity Texture UVs", parameterName="UV Channel (Opacity)", sliderMax=1}
  )
  n36 = AppendVector(
   A=ScalarParameter() {defaultValue=1, group="6.2) Opacity Texture UVs", parameterName="U Freq Mult (Opacity)", sortPriority=2}
   B=ScalarParameter() {defaultValue=1, group="6.2) Opacity Texture UVs", parameterName="V Freq Mult (Opacity)", sortPriority=3}
  )
  n37 = Panner(
   Coordinate=LinearInterpolate(
    A=Multiply(A=@n35, B=@n36) {desc=This might need a frac option for getting textures to tile. This scales the texture.}
    B=MaterialFunctionCall(UVs=@n35, Texture Scale=@n36) {materialFunction=&ScaleUVsByCenter}.UVs
    Alpha=ScalarParameter() {group="6.2) Opacity Texture UVs", parameterName="Scale UVs From Center (Opacity)", sortPriority=1}
   )
   Speed=AppendVector(
    A=ScalarParameter() {group="6.2) Opacity Texture UVs", parameterName="U Speed Mult (Distort)", sortPriority=4}
    B=ScalarParameter() {group="6.2) Opacity Texture UVs", parameterName="V Speed Mult (Distort)", sortPriority=5}
   )
  )
  n38 = TextureSampleParameter2D(UVs=LinearInterpolate(A=@n37, B=Add(A=@n37, B=@n14), Alpha=ScalarParameter() {group="6.2) Opacity Texture UVs", parameterName=Distort Opacity})) {
   group="6.0) Opacity Texture"
   parameterName=Opacity Texture
   sortPriority=1
   texture=/Game/Global/Textures/Default/Masks_and_Gradients/T_SparkRound
  }
  n39 = ChannelMaskParameter(None=@n38.RGB) {group="6.0) Opacity Texture", parameterName=Opacity Texture Mask, sortPriority=2}
  n40 = StaticSwitchParameter(
   True=@n34.V3 Length
   False=LinearInterpolate(
    A=@n39.R
    B=Multiply(A=@n34.V3 Length, B=@n39.R)
    Alpha=ScalarParameter() {group="6.0) Opacity Texture", parameterName=Multipy Opacity Texture Against Base Color, sliderMax=1, sortPriority=3}
   )
  ) {
   defaultValue=true
   desc="If Activated, you can use a seperate texture read for your opacity. Otherwise, it inherits from the basecolor material attribute."
   group="6.0) Opacity Texture"
   parameterName=Use Base Color for Opacity
  }
  n41 = TextureCoordinate()
  n42 = Power(
   Base=ComponentMask(None=OneMinus(None=@n41)) {r=true}
   Exp=ScalarParameter() {defaultValue=1, group="6.1) Opacity", parameterName=End Mask Softness, sortPriority=14}
  )
  n43 = ScalarParameter() {group="6.1) Opacity", parameterName=Mask Ribbon End, sortPriority=13}
  n44 = Multiply(
   A=@n40
   B=Multiply(
    A=ScalarParameter() {defaultValue=1, group="6.0) Opacity Texture", parameterName=Overall Opacity, sortPriority=1}
    B=LinearInterpolate(B=@n42, Alpha=@n43) {constA=1}
   )
  )
  n45 = Multiply(
   A=Multiply(A=Saturate(None=@n33), B=Saturate(None=@n44))
   B=LinearInterpolate(B=ParticleColor().A, Alpha=ScalarParameter() {group="1.0) Base Color Texture", parameterName=Use Particle Color/Alpha, sortPriority=1}) {constA=1}
  )
  n46 = MaterialFunctionCall!(In=SetMaterialAttributes!(
    MaterialAttributes=SetMaterialAttributes!(MaterialAttributes=@n23, Emissive Color=@n24.RemappedRange) {attributeSetTypes [#2]}
    Opacity=@n45
    Opacity Mask=@n45
   ) {attributeSetTypes [#3, #4]}) {materialFunction=/Game/Effects/Niagara/MasterMaterials/Functions/Opacity/FX_DepthFade}
  n47 = ScalarParameter() {defaultValue=13, group="7.0) World Position Offset", parameterName=WidthTip, sortPriority=1}
  n48 = ScalarParameter() {defaultValue=7, group="7.0) World Position Offset", parameterName=WidthBase, sortPriority=2}
  n49 = LinearInterpolate(
   A=TextureCoordinate()
   B=TextureCoordinate() {coordinateIndex=1}
   Alpha=ScalarParameter() {group="7.1) WPO UVs", parameterName="UV Channel (WPO)", sliderMax=1, sortPriority=1}
  )
  n50 = AppendVector(
   A=ScalarParameter() {defaultValue=1, group="7.1) WPO UVs", parameterName="Tile U (WPO)", sortPriority=2}
   B=ScalarParameter() {defaultValue=1, group="7.1) WPO UVs", parameterName="Tile V (WPO)", sortPriority=3}
  )
  n51 = DynamicParameter() {defaultValue=<rgba 0 0 0 1>, paramNames [R Offset, G Offset, B Offset, Speed - All]}
  n52 = AppendVector(
   A=ScalarParameter() {group="7.1) WPO UVs", parameterName="Pan U Speed (WPO)", sortPriority=4}
   B=ScalarParameter() {group="7.1) WPO UVs", parameterName="Pan V Speed (WPO)", sortPriority=5}
  )
  n53 = ScalarParameter() {group="7.1) WPO UVs", parameterName=Use Base Distort for WPO}
  n54 = LinearInterpolate(
   A=Panner(
    Coordinate=Multiply(A=@n49, B=@n50) {desc=This might need a frac option for getting textures to tile. This scales the texture.}
    Speed=Multiply(A=@n51.Speed - All, B=@n52)
   )
   B=@n12
   Alpha=@n53
  )
  n55 = TextureSample(UVs=@n54) {samplerType=SAMPLERTYPE_Normal, texture=&T_Weap_Distort_N_02}
  n56 = TextureSample(UVs=@n54) {texture=&T_Smooth_tile}
  n57 = AppendVector(
   A=ComponentMask(None=LinearInterpolate(A=@n55.RGB, B=@n13.RGB, Alpha=@n53)) {g=true, r=true}
   B=Add(A=Multiply(A=@n56.R) {constB=2}) {constB=-0.5}
  )
  n58 = ComponentMask(None=TextureCoordinate()) {r=true}
  n59 = VectorParameter() {
   desc=Alpha scales all values
   group="7.0) World Position Offset"
   parameterName=Scale WPO Intensity
   sortPriority=3
   defaultValue=<rgba 1 1 1 1>
  }
  n60 = ComponentMask(None=@n59.RGB) {b=true, g=true, r=true}
  n61 = Add(
   A=LinearInterpolate(
    B=MaterialFunctionCall(WidthBase=@n47, WidthTip=@n48) {materialFunction=&SplineThicken}.WorldPositionOffset
    Alpha=ScalarParameter() {defaultValue=1, group="7.0) World Position Offset", parameterName=Use Spline Thicken}
   )
   B=Multiply(A=LinearInterpolate(B=@n57, Alpha=@n58), B=Multiply(A=@n60, B=@n59.A)) {constB=10}
  )
  n62 = DynamicParameter() {defaultValue=<rgba 0 0 0 1>, paramNames [R Offset, G Offset, B Offset, Speed - All]}
  n63 = MaterialFunctionCall(R=@n62.R Offset, G=@n62.R Offset, B=@n62.R Offset, A=Constant()) {materialFunction=&AppendMany}
  n64 = ComponentMask(None=TextureCoordinate()) {r=true}
  n65 = ScalarParameter() {defaultValue=2, group="7.0) World Position Offset", parameterName=Edge Gradient Movement Control Exp, sortPriority=7}
  n66 = VectorParameter() {
   desc=Adjusts effective pivot if desired pivot point does not match the mesh
   group="7.0) World Position Offset"
   parameterName=Initial WPO Offset
   sortPriority=6
  }
  n67 = ComponentMask(None=TextureCoordinate()) {r=true}
  n68 = ScalarParameter() {defaultValue=2, group="7.0) World Position Offset", parameterName=Edge Gradient Movement Control Exp, sortPriority=7}
  n69 = Add(
   A=Multiply(A=Transform(None=@n63.RGBA) {transformSourceType=TRANSFORMSOURCE_Local}, B=Power(Base=@n64, Exp=@n65))
   B=Multiply(A=Transform(None=@n66.RGB) {transformSourceType=TRANSFORMSOURCE_Local}, B=Power(Base=@n67, Exp=@n68))
  )
  SetMaterialAttributes!(
   MaterialAttributes=GetMaterialAttributes(None=MaterialFunctionCall!(Material Attributes=@n46."Opacity (MA)") {materialFunction=/Game/Effects/Niagara/MasterMaterials/Functions/V2/MF_Skydive_CameraFade_V2}."Opacitys and BaseColor (MA)").MaterialAttributes
   World Position Offset=LinearInterpolate(
    A=@n61
    B=Add(A=@n61, B=@n69)
    Alpha=ScalarParameter() {group="7.0) World Position Offset", parameterName=Add Local Position Control, sortPriority=4}
   )
  ) {attributeSetTypes [#5]}
  n70 = Power(
   Base=MaterialFunctionCall(Float2=TextureCoordinate()) {materialFunction=&BreakOutFloat2Components}.R
   Exp=ScalarParameter() {defaultValue=2, group="7.0) World Position Offset", parameterName=Edge Gradient Movement Control Exp, sortPriority=7}
  )
  n71 = DynamicParameter() {defaultValue=<rgba 0 0 0 1>, paramNames [R Offset, G Offset, B Offset, Speed - All]}
  n72 = VectorParameter() {
   desc=Adjusts effective pivot if desired pivot point does not match the mesh
   group="7.0) World Position Offset"
   parameterName=Initial WPO Offset
   sortPriority=6
  }
  n73 = TransformPosition(None=WorldPosition().XYZ) {transformSourceType=TRANSFORMPOSSOURCE_World, transformType=TRANSFORMPOSSOURCE_Particle}
  n74 = RotateAboutAxis(
   NormalizedRotationAxis=Constant3Vector() {constant=<rgba 1 0 0 0>}
   RotationAngle=Multiply(A=@n70, B=@n71.R Offset)
   PivotPoint=@n72.RGB
   Position=@n73
  )
  n75 = RotateAboutAxis(
   NormalizedRotationAxis=Constant3Vector() {constant=<rgba 0 1 0 0>}
   RotationAngle=Multiply(A=@n70, B=@n71.G Offset)
   PivotPoint=@n72.RGB
   Position=@n73
  )
  n76 = Constant3Vector() {constant=<rgba 0 0 1 0>}
  n77 = Multiply(A=@n70, B=@n71.B Offset)
  TransformPosition(None=Add(A=Add(A=@n74, B=@n75), B=RotateAboutAxis(NormalizedRotationAxis=@n76, RotationAngle=@n77, PivotPoint=@n72.RGB, Position=@n73))) {transformSourceType=TRANSFORMPOSSOURCE_Particle, transformType=TRANSFORMPOSSOURCE_World}
 ]
}
