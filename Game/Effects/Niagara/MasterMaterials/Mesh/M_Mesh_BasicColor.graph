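# Niagara master material for lit mesh and sprite particles that drives color and emissive from particle color.
# A static switch (Use_MeshGlow) selects either a particle-tinted two-layer noise glow or particle color scaled by Add Emissive, feeding an emissive adjustments function.
# The glow adds two tinted texture layers, one UV-distorted by panning noise and the other modulated by panning cell noise.
# Diffuse Mult scales particle color separately, alongside an MSR vector parameter.
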Material /Game/Effects/Niagara/MasterMaterials/Mesh/M_Mesh_BasicColor
graph {
 properties {bUsedWithNiagaraMeshParticles=true, bUsedWithNiagaraSprites=true, pixelDepthOffsetMode=PDOM_Legacy, usedShadingModels=Default Lit}
 nodes [
  n1 = ParticleColor()
  n2 = TextureCoordinate()
  n3 = AppendVector(A=ScalarParameter() {parameterName=BaseTextureU_add}, B=ScalarParameter() {parameterName=BaseTextureV_add})
  n4 = TextureSample(UVs=Panner(Coordinate=TextureCoordinate() {uTiling=2, vTiling=3}) {speedX=-0.5}) {samplerType=SAMPLERTYPE_LinearColor, texture=&T_NoisePack_LinearColor}
  n5 = ScalarParameter() {defaultValue=0.02, parameterName=Noise Intensity}
  n6 = TextureSampleParameter2D(UVs=Add(A=Add(A=@n2, B=@n3), B=Multiply(A=@n4.R, B=@n5) {constB=0.02})) {parameterName=BaseTexture, texture=/Game/Effects/Fort_Effects/Emotes/UltraLight/Texture/T_SoftNoise_BC7}
  n7 = Power(Base=TextureSampleParameter2D() {parameterName=BaseTexture02, texture=&T_SoftNoise_BC7}.G) {constExponent=1.2}
  n8 = TextureCoordinate()
  n9 = LinearInterpolate(Alpha=TextureSample(UVs=Panner(Coordinate=@n8) {speedX=0.3}) {samplerType=SAMPLERTYPE_Grayscale, texture=/Game/Weapons/FORT_Melee/Pickaxe_Cube_Ninja_Male/FX/T_CellNoise}.R)
  n10 = Add(
   A=Multiply(A=VectorParameter() {parameterName=color, defaultValue=<rgba 0 0.1677 1.5 1>}.RGB, B=Power(Base=@n6.R) {constExponent=1.5})
   B=Multiply(A=VectorParameter() {parameterName=color02, defaultValue=<rgba 15 0.3906 0.6311 1>}.RGB, B=Multiply(A=@n7, B=@n9))
  )
  n11 = Constant()
  n12 = ScalarParameter() {parameterName=Add Emissive}
  MaterialFunctionCall(Emissive=StaticSwitchParameter(True=Multiply(A=@n1.RGB, B=@n10), False=LinearInterpolate(A=@n11, B=@n1.RGB, Alpha=@n12)) {parameterName=Use_MeshGlow}) {materialFunction=&MF_FX_EmissiveAdjustments}
  Multiply(A=@n1.RGB, B=ScalarParameter() {defaultValue=1, desc=Diffuse Contribution, parameterName=Diffuse Mult, sliderMax=1})
  VectorParameter() {parameterName=MSR, defaultValue=<rgba 0 0 0 1>}
 ]
}
