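# Master material for Niagara and mesh particles, named for debris, using opaque-default settings and Default Lit shading with material attributes.
# Chained attribute setters apply a diffuse texture, a normal map with adjustable flattening, and a mask texture whose green channel drives Metallic, Specular and Roughness alike.
# A final stage multiplies Base Color and Emissive by particle color, blended from white via UseParticleColor, though no emissive input is authored upstream.
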Material /Game/Effects/Niagara/MasterMaterials/Mesh/M_Mesh_Debirs
graph {
 properties {
  bAutomaticallySetUsageInEditor=false
  bUseMaterialAttributes=true
  bUsedWithMeshParticles=true
  bUsedWithNiagaraMeshParticles=true
  pixelDepthOffsetMode=PDOM_Legacy
  usedShadingModels=Default Lit
 }
 nodes [
  n1 = TextureSampleParameter2D() {parameterName=SpecularMasks, samplerType=SAMPLERTYPE_Masks, texture=&T_Fortnite_Default_S}
  n2 = TextureSampleParameter2D() {parameterName=Normals, samplerType=SAMPLERTYPE_Normal, texture=/Engine/EngineMaterials/DefaultNormal}
  n3 = ScalarParameter() {parameterName=FlattenNormal}
  n4 = SetMaterialAttributes!(
   MaterialAttributes=SetMaterialAttributes!(
    MaterialAttributes=SetMaterialAttributes!(MaterialAttributes, Metallic=@n1.G, Specular=@n1.G, Roughness=@n1.G) {attributeSetTypes [#1, #2, #3]}
    Normal=MaterialFunctionCall(Normal=@n2.RGB, Flatness=@n3) {materialFunction=&FlattenNormal}.Result
   ) {attributeSetTypes [#4]}
   Base Color=TextureSampleParameter2D() {parameterName=Diffuse, texture=/Engine/EngineMaterials/DefaultDiffuse}.RGB
  ) {attributeSetTypes [#5]}
  n5 = GetMaterialAttributes(None=@n4) {attributeGetTypes [#5, #6]}
  n6 = LinearInterpolate(A=Constant() {r=1}, B=ParticleColor().RGB, Alpha=ScalarParameter() {defaultValue=1, parameterName=UseParticleColor})
  SetMaterialAttributes!(MaterialAttributes=@n4, Base Color=Multiply(A=@n5.Base Color, B=@n6), Emissive Color=Multiply(A=@n5.Emissive Color, B=@n6)) {attributeSetTypes [#5, #6]}
 ]
}
