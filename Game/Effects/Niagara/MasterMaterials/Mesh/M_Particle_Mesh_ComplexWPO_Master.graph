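# Master masked, two-sided lit material for Niagara mesh particles, with complex world position offset.
# Diffuse or SubUV color takes optional color-chart tinting by object position and hue shift, optionally feeding masked emissive.
# Particle alpha drives texture-erosion dissolve, combined with near-camera fade.
# WPO sums cloth animation, vertex-color Z scaling, distorted panning waves and randomized directional movement, modulated by dynamic parameters, before all attributes pass through MF_CreativeMoveTool.
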Material /Game/Effects/Niagara/MasterMaterials/Mesh/M_Particle_Mesh_ComplexWPO_Master
graph {
 properties {
  bUseMaterialAttributes=true
  bUsedWithMeshParticles=true
  blendMode=BLEND_Masked
  opacityMaskClipValue=0.16
  pixelDepthOffsetMode=PDOM_Legacy
  twoSided=true
  usedShadingModels=Default Lit
 }
 nodes [
  n1 = VectorParameter() {group=World Space Color Picker, parameterName=World Space Color Picker Color Override, sortPriority=4, defaultValue=<rgba 1 1 1 1>}
  n2 = ScalarParameter() {
   defaultValue=0.1
   desc="1/number of colors. ie 1/10 = .1"
   group=World Space Color Picker
   parameterName="Number Of Colors in Color Chart (1/Number of colors)"
   sortPriority=8
  }
  n3 = Multiply(A=ComponentMask(None=TextureCoordinate() {coordinateIndex=1}) {g=true}, B=@n2)
  n4 = Divide(
   A=MaterialFunctionCall!(f2, f3=ObjectPositionWS(), f4) {materialFunction=/Game/Packages/Fortnite_UniqueMaterials_1/VectorComponentOps}.f3 Add
   B=ScalarParameter() {defaultValue=16.71, group=World Space Color Picker, parameterName="Color Change Frequency (WS)", sortPriority=7}
  )
  n5 = TextureSampleParameter2D(UVs=StaticSwitchParameter(True=Add(A=@n3, B=@n4), False=@n4) {group=World Space Color Picker, parameterName=Offset Color Index with UV Channel 2 G, sortPriority=6}) {
   group=World Space Color Picker
   parameterName=Color Chart
   sortPriority=5
   texture=/Game/Global/Textures/Default/ColorSets/CarColors_Interiors
  }
  n6 = ScalarParameter() {defaultValue=1, group=World Space Color Picker, parameterName=Use Color Chart, sortPriority=3}
  n7 = TextureSampleParameter2D() {
   group=Diffuse
   parameterName=Diffuse
   samplerSource=SSM_Wrap_WorldGroupSettings
   sortPriority=1
   texture=/Engine/EngineMaterials/DefaultDiffuse
  }
  n8 = TextureSampleParameterSubUV!(UVs, Apply View MipBias) {
   group=Diffuse
   parameterName=Diffuse_SubUV
   samplerSource=SSM_Wrap_WorldGroupSettings
   sortPriority=2
   texture=&DefaultDiffuse
  }
  n9 = ScalarParameter() {group=Diffuse, parameterName=Use SubUV Diffuse Texture}
  n10 = LinearInterpolate(A=@n7.RGB, B=@n8.RGB, Alpha=@n9)
  n11 = TextureSampleParameter2D() {
   group=World Space Color Picker
   parameterName=World Space Color Picker Mask
   samplerSource=SSM_Wrap_WorldGroupSettings
   sortPriority=1
   texture=&T_BlacksRGB
  }
  n12 = ChannelMaskParameter(None=@n11.RGBA) {
   group=World Space Color Picker
   maskChannel=Blue
   parameterName=World Space Color Picker Mask Channel
   sortPriority=2
   defaultValue=<rgba 0 0 1 0>
  }
  n13 = LinearInterpolate(A=Multiply(A=LinearInterpolate(A=@n1.RGB, B=@n5.RGB, Alpha=@n6), B=@n10), B=@n10, Alpha=@n12.R)
  n14 = ScalarParameter() {group=World Space Color Picker, parameterName=Use World Position Coloring}
  n15 = MaterialFunctionCall(
   Hue Shift Percentage=ScalarParameter() {group=Diffuse, parameterName=Hue Shift Percentage, sortPriority=4}
   Texture=Multiply(
    A=LinearInterpolate(A=@n10, B=@n13, Alpha=@n14)
    B=ScalarParameter() {defaultValue=1, group=Diffuse, parameterName=Diff_Mult, sortPriority=3}
   )
  ) {materialFunction=&HueShift}
  n16 = TextureSampleParameter2D() {
   group=Specular
   parameterName=SpecMasks
   samplerSource=SSM_Wrap_WorldGroupSettings
   samplerType=SAMPLERTYPE_Masks
   texture=&T_Fortnite_Default_S
  }
  n17 = ScalarParameter() {group=Specular, parameterName=SpecRoughnessMin, sortPriority=1}
  n18 = ScalarParameter() {defaultValue=1, group=Specular, parameterName=SpecRoughnessMax, sortPriority=2}
  n19 = VectorParameter() {group=Emissive, parameterName="Emissive Color && Intensity", sortPriority=2}
  n20 = DynamicParameter() {defaultValue=<rgba 0 1 0 4>, paramNames [Emissive Intensity, UV Distortion Multiplier, Multiplier2, Directional Movement Lerp]}
  n21 = Multiply(
   A=LinearInterpolate(
    A=@n19.RGB
    B=Multiply(A=@n15.Result, B=@n19.RGB)
    Alpha=ScalarParameter() {group=Emissive, parameterName=UseDiffuseinEmissive, sortPriority=1}
   )
   B=@n20.Emissive Intensity
  )
  n22 = TextureSampleParameter2D() {
   group=Emissive
   parameterName=Emissive Mask Texture
   samplerSource=SSM_Wrap_WorldGroupSettings
   sortPriority=3
   texture=&DefaultDiffuse
  }
  n23 = LinearInterpolate(
   A=ChannelMaskParameter(None=@n22.RGBA) {group=Emissive, parameterName=Emissive Mask Texture Channel, sortPriority=4}.R
   B=@n22.RGB
   Alpha=ScalarParameter() {group=Emissive, parameterName=Emissive Mask Uses RGB, sortPriority=5}
  )
  n24 = LinearInterpolate(A=@n23, B=OneMinus(None=@n23), Alpha=ScalarParameter() {group=Emissive, parameterName=Invert Emissive Mask, sortPriority=6})
  n25 = TextureSampleParameter2D() {
   group=Normal
   parameterName=Normals
   samplerSource=SSM_Wrap_WorldGroupSettings
   samplerType=SAMPLERTYPE_Normal
   texture=/Engine/EngineMaterials/DefaultNormal
  }
  n26 = VertexColor()
  n27 = ChannelMaskParameter(None=AppendVector(A=@n26, B=@n26)) {group=WPO_ClothAnimation, parameterName=Cloth_Vertex_Pan_Speed_Offset_Color}
  n28 = ScalarParameter() {defaultValue=1, group=WPO_ClothAnimation, parameterName=WPO_ClothPanSpeed}
  n29 = MaterialFunctionCall!(
   UVs
   Pan Speed=Clamp(
    None=StaticSwitchParameter(True=@n27.R, False=@n28) {group=WPO_ClothAnimation, parameterName=Use_Offset_Cloth_By_Vertex_Color}
    Min=ScalarParameter() {group=WPO_ClothAnimation, parameterName=WPO_ClothPanSpeedMin}
   )
   WPO_ClothOffsetDirection
   WPO_ClothNormalIntensity
   WPO_ClothOffsetIntensity
   "Motion Mask (Defaults to the Red Vertex Color)"
   WPO_ClothWPOWindTextureOffset
   Pan Direction
  ) {materialFunction=/Game/Packages/Fortress_SharedMaterials/Base_Material_Functions/Cloth}
  n30 = ScalarParameter() {parameterName=Use WPO_ClothAnimation}
  n31 = LinearInterpolate(B=@n29.Normal, Alpha=@n30)
  n32 = ScalarParameter() {group=Opacity, parameterName=Near Camera Fade Distance}
  n33 = LinearInterpolate(
   B=LinearInterpolate(A=@n7.A, B=@n8.A, Alpha=@n9)
   Alpha=ScalarParameter() {defaultValue=1, group=Diffuse, parameterName=Diffuse Alpha Blend, sortPriority=3}
  ) {constA=1}
  n34 = TextureCoordinate()
  n35 = ComponentMask(None=VectorParameter() {group=Opacity, parameterName=Erosion Mask UV Scale, sortPriority=3, defaultValue=<rgba 1 1 0 0>}.RGB) {g=true, r=true}
  n36 = ParticleColor()
  n37 = TextureSampleParameter2D(UVs=Multiply(A=Multiply(A=@n34, B=@n35), B=@n36.R)) {group=Opacity, parameterName=ErodeTexture, sortPriority=1, texture=&T_Inky_Smoke_Tile}
  n38 = MaterialFunctionCall!(
   Ramp=ChannelMaskParameter(None=@n37.RGBA) {group=Opacity, parameterName=Erode Mask, sortPriority=2}.R
   Dissolve=OneMinus(None=@n36.A)
   EdgeSharpness
  ) {materialFunction=/Game/Effects/Fort_Effects/Materials/_Functions/MF_Dissolve}
  n39 = Multiply(
   A=Multiply(
    A=MaterialFunctionCall(Distance=@n32) {materialFunction=&Near_Camera_Fade}.Result
    B=Multiply(A=@n33, B=@n38.ResultWithOutClamp)
   )
   B=ParticleColor().A
  )
  n40 = DynamicParameter() {defaultValue=<rgba 0 1 0 4>, paramNames [Scale_Z, UV Distortion Multiplier, Multiplier2, Directional Movement Lerp]}
  n41 = Multiply(
   A=Multiply(A=@n40.Scale_Z, B=ScalarParameter() {group=WPO_ClothAnimation, parameterName=Scale_Z})
   B=Multiply(A=VertexColor(), B=VectorParameter() {parameterName=Vertex Color Channel To Multiply Against, defaultValue=<rgba 1 0 1 1>}.RGB)
  )
  n42 = Constant()
  n43 = WorldPosition()
  n44 = ObjectPositionWS()
  n45 = MaterialFunctionCall(X=@n42, Y=@n42, Z=ComponentMask(None=Subtract(A=@n43.XYZ, B=@n44)) {b=true}) {materialFunction=&MakeFloat3}
  n46 = TextureSampleParameter2D(UVs=TextureCoordinate() {vTiling=2}) {group=WPO_ClothAnimation, parameterName=UV Distortion Texture, texture=&T_Inky_Smoke_Tile}
  n47 = DynamicParameter() {defaultValue=<rgba 0 1 0 4>, paramNames [Scale_Z, UV Distortion Multiplier, Multiplier2, Directional Movement Lerp]}
  n48 = Add(A=TextureCoordinate() {vTiling=4}, B=Multiply(A=@n46.G, B=@n47.UV Distortion Multiplier))
  n49 = Frac(None=ComponentMask(None=Panner(Coordinate=@n48) {speedY=1}) {g=true})
  n50 = Multiply(A=@n49, B=OneMinus(None=@n49))
  n51 = DynamicParameter() {defaultValue=<rgba 0 1 0 4>, paramNames [Scale_Z, UV Distortion Multiplier, Multiplier2, Directional Movement Lerp]}
  n52 = Add(
   A=Add(A=LinearInterpolate(B=@n29.Additive WPO, Alpha=@n30), B=Multiply(A=@n41, B=@n45.Result) {constB=-1})
   B=Multiply(A=Power(Base=@n50), B=@n51.Multiplier2) {constB=1500}
  )
  n53 = WorldPosition()
  n54 = Add(A=Add(A=ComponentMask(None=@n53.XYZ) {r=true}, B=ComponentMask(None=@n53.XYZ) {r=true}), B=ComponentMask(None=@n53.XYZ) {r=true})
  n55 = MaterialFunctionCall!(
   Random Picker=ComponentMask(None=Divide(A=@n54) {constB=4096}) {r=true}
   Texture Input
   Option 1=Constant3Vector() {constant=<rgba -1 1 1 0>}
   Option 2=Constant3Vector() {constant=<rgba 1 -1 1 0>}
   Option 3=Constant3Vector() {constant=<rgba 1 -1 -1 0>}
  ) {materialFunction=/Game/Effects/Fort_Effects/Materials/_Functions/MF_RGB_Random_01}
  n56 = VectorParameter() {parameterName=Random Direction Multiplier, defaultValue=<rgba 512 512 512 1>}
  n57 = DynamicParameter() {defaultValue=<rgba 0 1 0 4>, paramNames [Scale_Z, UV Distortion Multiplier, Multiplier2, Directional Movement Lerp]}
  n58 = LinearInterpolate(B=Multiply(A=VertexColor().R, B=Multiply(A=@n55.Vec3 Out Only, B=@n56.RGB) {constB=512}), Alpha=@n57.Directional Movement Lerp)
  MaterialFunctionCall(MA In=MakeMaterialAttributes!(
    BaseColor=@n15.Result
    Metallic=@n16.G
    Specular=@n16.G
    Roughness=LinearInterpolate(A=@n17, B=@n18, Alpha=@n16.B)
    Anisotropy
    EmissiveColor=Multiply(A=@n21, B=@n24)
    Opacity=@n39
    OpacityMask=@n39
    Normal=MaterialFunctionCall(BaseNormal=@n25.RGB, AdditionalNormal=@n31) {materialFunction=&BlendAngleCorrectedNormals}.Result
    Tangent
    WorldPositionOffset=Add(A=@n52, B=@n58)
    SubsurfaceColor
    ClearCoat
    ClearCoatRoughness
    AmbientOcclusion
    Refraction
    CustomizedUVs_0
    CustomizedUVs_1
    CustomizedUVs_2
    CustomizedUVs_3
    CustomizedUVs_4
    CustomizedUVs_5
    CustomizedUVs_6
    CustomizedUVs_7
    PixelDepthOffset
    ShadingModel
    Displacement
   )) {materialFunction=&MF_CreativeMoveTool}
 ]
}
