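# Translucent, two-sided, lit master material for Niagara mesh particles, with scrolling noise opacity and Fresnel.
# Particle color, optionally tinted by a panning texture, feeds an emissive adjustment function.
# A panning base mask is multiplied by two toggleable UV1 noise masks, plus an adjustable base-alpha term.
# Fresnel adds to or multiplies that result, particle alpha scales it, and depth fade applies except on mobile.
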Material /Game/Effects/Niagara/MasterMaterials/Mesh/M_Particle_Mesh_Noise
graph {
 properties {
  bUsedWithNiagaraMeshParticles=true
  blendMode=BLEND_Translucent
  pixelDepthOffsetMode=PDOM_Legacy
  twoSided=true
  usedShadingModels=Default Lit
 }
 nodes [
  n1 = ParticleColor()
  n2 = TextureCoordinate()
  n3 = AppendVector(
   A=ScalarParameter() {defaultValue=1, group="1) Emissive", parameterName=Emissive-Tiling_X}
   B=ScalarParameter() {defaultValue=1, group="1) Emissive", parameterName=Emissive-Tiling_Y}
  )
  n4 = ScalarParameter() {group="1) Emissive", parameterName=Emissive-Speed_X}
  n5 = ScalarParameter() {group="1) Emissive", parameterName=Emissive-Speed_Y}
  n6 = TextureSampleParameter2D(UVs=Panner(Coordinate=Multiply(A=@n2, B=@n3), Speed=AppendVector(A=@n4, B=@n5))) {group="1) Emissive", parameterName=Emissive_Texture, texture=/Game/Global/Textures/Default/Blanks/T_WhiteUserInterface}
  n7 = LinearInterpolate(
   A=ChannelMaskParameter(None=@n6.RGB) {group="1) Emissive", parameterName=Emissive-Texture_ChannelPicker}.R
   B=@n6.RGB
   Alpha=ScalarParameter() {group="1) Emissive", parameterName=USE_EMISSIVE_RGB}
  )
  MaterialFunctionCall(Emissive=LinearInterpolate(A=@n1.RGB, B=Multiply(A=@n7, B=@n1.RGB), Alpha=ScalarParameter() {group="1) Emissive", parameterName=Use_Emissive_Texture, sortPriority=-1})) {materialFunction=&MF_FX_EmissiveAdjustments}
  n8 = MaterialFunctionCall(
   Brightness=ScalarParameter() {defaultValue=1, group=Fresnel, parameterName=Fresnel_Brightness}
   Exponent=ScalarParameter() {defaultValue=20, group=Fresnel, parameterName=Fresnel_Exp}
   Vertex / Pixel Normal=VertexNormalWS()
  ) {materialFunction=&MF_Fresnel_FX}
  n9 = ScalarParameter() {defaultValue=1, group=Fresnel, parameterName=Double Sided Fresnel}
  n10 = ScalarParameter() {defaultValue=1, group=Fresnel, parameterName=Double Sided Fresnel}
  n11 = Multiply(
   A=LinearInterpolate(
    A=LinearInterpolate(A=@n8.Fresnel, B=@n8.Fresnel, Alpha=@n9)
    B=LinearInterpolate(A=@n8.Fresnel Inverted, B=@n8.Fresnel Inverted, Alpha=@n10)
    Alpha=ScalarParameter() {defaultValue=1, group=Fresnel, parameterName=InvertFresnel}
   )
   B=ScalarParameter() {defaultValue=0.1, group=Fresnel, parameterName=Fresnel_Multiplier}
  )
  n12 = TextureCoordinate()
  n13 = AppendVector(
   A=ScalarParameter() {defaultValue=1, group=BaseOpacityMask, parameterName=OpacityBase-Tiling_X}
   B=ScalarParameter() {defaultValue=1, group=BaseOpacityMask, parameterName=OpacityBase-Tiling_Y}
  )
  n14 = ScalarParameter() {group=BaseOpacityMask, parameterName=OpacityBase-Speed_X}
  n15 = ScalarParameter() {group=BaseOpacityMask, parameterName=OpacityBase-Speed_Y}
  n16 = TextureSampleParameter2D(UVs=Panner(Coordinate=Multiply(A=@n12, B=@n13), Speed=AppendVector(A=@n14, B=@n15))) {group=BaseOpacityMask, parameterName=BaseOpacityMask, texture=/Game/Effects/Fort_Effects/Emotes/Alchemy/T_FX_Alchemy_HeartHologram_Mask}
  n17 = ChannelMaskParameter(None=@n16.RGB) {
   desc="Chooses Texture Channel. For some reason the alpha selector does not work. Use \"USE_ALPHA\" instead."
   group=BaseOpacityMask
   maskChannel=Alpha
   parameterName=Base_Opacity_ChannelPicker
   defaultValue=<rgba 0 0 0 1>
  }
  n18 = Clamp(None=LinearInterpolate(A=@n17.R, B=@n16.A, Alpha=ScalarParameter() {group=BaseOpacityMask, parameterName=USE_ALPHA}))
  n19 = TextureCoordinate() {coordinateIndex=1}
  n20 = AppendVector(
   A=ScalarParameter() {defaultValue=2, group=OpacityMask-1, parameterName=OpacityMask-1:Tiling-X}
   B=ScalarParameter() {defaultValue=2, group=OpacityMask-1, parameterName=OpacityMask-1:Tiling-Y}
  )
  n21 = ScalarParameter() {defaultValue=-0.1, group=OpacityMask-1, parameterName=OpacityMask-1:Speed-X}
  n22 = ScalarParameter() {defaultValue=0.25, group=OpacityMask-1, parameterName=OpacityMask-1:Speed-Y}
  n23 = TextureSampleParameter2D(UVs=Panner(Coordinate=Multiply(A=@n19, B=@n20), Speed=AppendVector(A=@n21, B=@n22))) {
   group=OpacityMask-1
   parameterName=OpacityMask-1:Texture
   texture=/Game/Characters/Player/Male/Medium/Bodies/M_MED_Viper/Textures/Glitch/T_SoftNoise
  }
  n24 = Multiply(A=@n23.B, B=ScalarParameter() {defaultValue=2, group=OpacityMask-1, parameterName=OpacityMask-1:Mult})
  n25 = ScalarParameter() {defaultValue=2, group=OpacityMask-1, parameterName=OpacityMask-1:Exp}
  n26 = LinearInterpolate(
   B=Clamp(None=Power(Base=@n24, Exp=@n25))
   Alpha=ScalarParameter() {defaultValue=1, group=OpacityMask-1, parameterName=Use_OpacityMask_1, sortPriority=-1}
  ) {constA=1}
  n27 = Multiply(
   A=TextureCoordinate() {coordinateIndex=1}
   B=AppendVector(
    A=ScalarParameter() {defaultValue=10, group=OpacityMask-2, parameterName=OpacityMask-2:Tiling-X}
    B=ScalarParameter() {group=OpacityMask-2, parameterName=OpacityMask-2:Tiling-Y}
   )
  )
  n28 = AppendVector(
   A=ScalarParameter() {defaultValue=-0.01, group=OpacityMask-2, parameterName=OpacityMask-2:Speed-X}
   B=ScalarParameter() {defaultValue=0.1, group=OpacityMask-2, parameterName=OpacityMask-2:Speed-Y}
  )
  n29 = Multiply(
   A=TextureSampleParameter2D(UVs=Panner(Coordinate=@n27, Speed=@n28)) {group=OpacityMask-2, parameterName=OpacityMask-2:Texture, texture=&AtmosphericCloudNoise01}.B
   B=ScalarParameter() {defaultValue=2, group=OpacityMask-2, parameterName=OpacityMask-2:Mult}
  )
  n30 = ScalarParameter() {defaultValue=2, group=OpacityMask-2, parameterName=OpacityMask-2:Exp}
  n31 = LinearInterpolate(
   B=Clamp(None=Power(Base=@n29, Exp=@n30))
   Alpha=ScalarParameter() {defaultValue=1, group=OpacityMask-2, parameterName=Use_OpacityMask_2, sortPriority=-1}
  ) {constA=1}
  n32 = Add(A=Multiply(A=@n18, B=Multiply(A=@n26, B=@n31)), B=Multiply(A=@n18, B=ScalarParameter() {parameterName=BASE_ALPHA_STRENGTH}))
  n33 = Multiply(
   A=@n1.A
   B=LinearInterpolate(
    A=Add(A=@n11, B=@n32)
    B=Multiply(A=@n11, B=@n32)
    Alpha=ScalarParameter() {group=Fresnel, parameterName=Add-Or-Multiply_Fresnel, sortPriority=-1}
   )
  )
  ShadingPathSwitch(
   Default=DepthFade(Opacity=@n33, FadeDistance=ScalarParameter() {defaultValue=10, group=BaseOpacityMask, parameterName=DepthFade_Distance, sortPriority=-1})
   Mobile=@n33
  )
 ]
}
