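# Translucent unlit master material for Niagara mesh and sprite particles, driven by particle color.
# Opacity combines particle alpha, an optional noise-modulated linear UV mask, Fresnel and depth fade.
# An optional dissolve multiplies in a UV1 sphere mask mixed with panning noise via dynamic parameters; both pan axes read the U speed.
# Emissive optionally applies Fresnel brightening before MF_FX_EmissiveAdjustments.
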Material /Game/Effects/Niagara/MasterMaterials/Mesh/M_Particle_Mesh_Unlit_Master
graph {
 properties {
  bUsedWithMeshDeformer=true
  bUsedWithNiagaraMeshParticles=true
  bUsedWithNiagaraSprites=true
  bUsedWithSkeletalMesh=true
  blendMode=BLEND_Translucent
  pixelDepthOffsetMode=PDOM_Legacy
  shadingModel=MSM_Unlit
  usedShadingModels=Unlit
 }
 nodes [
  n1 = ParticleColor()
  n2 = OneMinus(None=ComponentMask(None=TextureCoordinate()) {g=true})
  n3 = Multiply(A=@n2, B=TextureSample() {samplerType=SAMPLERTYPE_Grayscale, texture=&LowResBlurredNoise}.R)
  n4 = Saturate(None=Power(Base=Add(A=@n2, B=@n3), Exp=ScalarParameter() {defaultValue=2, parameterName=LinearMask_Power}))
  n5 = Multiply(
   A=LinearInterpolate(A=@n1.A, B=Multiply(A=@n1.A, B=@n4), Alpha=ScalarParameter() {parameterName=Use_LinearMask})
   B=MaterialFunctionCall(
    Invert Fresnel=StaticBoolParameter() {defaultValue=true, parameterName=Invert Fresnel}
    Power=ScalarParameter() {defaultValue=2, parameterName=Fresnel Power}
   ) {materialFunction=&Fresnel_Function}.Result
  )
  n6 = ScalarParameter() {defaultValue=1, parameterName=Fresnel_Multiply}
  n7 = DepthFade(Opacity=Saturate(None=Multiply(A=@n5, B=@n6) {constB=3}), FadeDistance=ScalarParameter() {defaultValue=100, parameterName=FadeDistance})
  n8 = DynamicParameter() {paramNames [Dissolve_Mix, Dissolve_Process, "", ""]}
  n9 = SphereMask(A=ComponentMask(None=TextureCoordinate() {coordinateIndex=1}) {g=true}, B=@n8.Dissolve_Process, Radius=Constant() {r=2}) {attenuationRadius=1, hardnessPercent=0}
  n10 = TextureCoordinate() {coordinateIndex=1}
  n11 = DynamicParameter() {
   parameterIndex=1
   defaultValue=<rgba 0.5 0.3 0 0.2>
   paramNames [Dissolve Tile U, Dissolve Tile V, Dissolve Pan Speed U, Dissolve Pan Speed V]
  }
  n12 = Multiply(A=ComponentMask(None=@n10) {r=true}, B=@n11.Dissolve Tile U)
  n13 = Multiply(A=ComponentMask(None=@n10) {g=true}, B=@n11.Dissolve Tile V)
  n14 = TextureSampleParameter2D(UVs=Panner(Coordinate=AppendVector(A=@n12, B=@n13), Speed=AppendVector(A=@n11.Dissolve Pan Speed U, B=@n11.Dissolve Pan Speed U)) {speedX=-1, speedY=0.1}) {
   group=Dissolve
   parameterName=Dissolve Tex
   samplerType=SAMPLERTYPE_Grayscale
   texture=&LowResBlurredNoise
  }
  n15 = Multiply(A=@n9, B=@n14.R)
  LinearInterpolate(
   A=@n7
   B=Multiply(A=@n7, B=LinearInterpolate(A=@n9, B=@n15, Alpha=@n8.Dissolve_Mix) {constAlpha=0.7})
   Alpha=ScalarParameter() {parameterName=Ues Dissolve}
  )
  n16 = LinearInterpolate(Alpha=Fresnel(
    ExponentIn=ScalarParameter() {defaultValue=3, parameterName=EmissiveFresnel_Exp}
    BaseReflectFractionIn=ScalarParameter() {parameterName=EmissiveFresnel_BRF}
   ) {baseReflectFraction=0, exponent=3}) {constA=0.5, constB=10}
  MaterialFunctionCall(Emissive=LinearInterpolate(A=@n1.RGB, B=Multiply(A=@n1.RGB, B=@n16), Alpha=ScalarParameter() {parameterName=Use_EmissiveFresnel})) {materialFunction=&MF_FX_EmissiveAdjustments}
 ]
}
