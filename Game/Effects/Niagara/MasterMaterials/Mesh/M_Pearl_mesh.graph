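# Masked Default Lit master material that gives Niagara mesh and sprite particles a two-tone, pearl-like glow.
# A clamped Fresnel term blends the Inner Color and Outer Color parameters, and a dimmed copy of the blend is also produced.
# The emissive sum adds a high-power, Fresnel-weighted copy of the blend to the blend tinted by particle color.
# That sum also feeds an MF_FX_EmissiveAdjustments call.
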Material /Game/Effects/Niagara/MasterMaterials/Mesh/M_Pearl_mesh
graph {
 properties {
  bUsedWithNiagaraMeshParticles=true
  bUsedWithNiagaraSprites=true
  blendMode=BLEND_Masked
  pixelDepthOffsetMode=PDOM_Legacy
  previewMesh=/Game/Gadgets/Assets/VinderTech_GliderChute/Glider_TeddyBear/Effects/Meshes/SM_Heart_3D
  usedShadingModels=Default Lit
 }
 nodes [
  Constant()
  n1 = Fresnel() {exponent=0.5}
  n2 = Clamp(None=Multiply(A=Power(Base=@n1) {constExponent=1}) {constB=1.25})
  n3 = LinearInterpolate(
   A=VectorParameter() {parameterName=Inner Color, defaultValue=<rgba 0.2852 0.0127 1.5 1>}.RGB
   B=VectorParameter() {parameterName=Outer Color, defaultValue=<rgba 0 0.1798 0.25 1>}.RGB
   Alpha=@n2
  )
  n4 = Power(Base=@n3) {constExponent=10}
  n5 = Add(A=Multiply(A=Multiply(A=@n4) {constB=3}, B=@n2), B=Multiply(A=@n3, B=ParticleColor().RGB))
  MaterialFunctionCall(Emissive=@n5) {materialFunction=&MF_FX_EmissiveAdjustments}
  Reroute(None=Multiply(A=@n3) {constB=0.3})
  Reroute(None=Constant())
  Reroute(None=Constant() {r=1})
  Reroute(None=Constant() {r=0.2})
  Reroute(None=@n5)
 ]
}
