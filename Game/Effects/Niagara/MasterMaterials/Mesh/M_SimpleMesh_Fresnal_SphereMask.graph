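# Unlit master material for Niagara mesh particles, combining particle color with a spherical mask around each particle's pivot.
# Particle color RGB feeds the shared MF_FX_EmissiveAdjustments function for emissive shaping.
# Particle alpha is multiplied by a small, soft SphereMask that compares world position to the mesh particle pivot, fading with distance from it.
# Despite the name, the graph has no Fresnel node of its own.
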Material /Game/Effects/Niagara/MasterMaterials/Mesh/M_SimpleMesh_Fresnal_SphereMask
graph {
 properties {
  bAutomaticallySetUsageInEditor=false
  previewMesh=/BRCosmetics/Gadgets/Assets/VinderTech_Umbrella/Umbrella_Default/SM_FX_Unhinged_UmbrellaFXMesh_Heart
  shadingModel=MSM_Unlit
  usedShadingModels=Unlit
 }
 nodes [
  n1 = ParticleColor()
  MaterialFunctionCall(Emissive=@n1.RGB) {materialFunction=&MF_FX_EmissiveAdjustments}
  Multiply(
   A=@n1.A
   B=SphereMask(
    A=WorldPosition().XYZ
    B=MaterialFunctionCall() {materialFunction=&ObjectPivotPoint}.Mesh Particle Pivot Location
   ) {attenuationRadius=12, hardnessPercent=1}
  )
 ]
}
