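# Translucent, unlit master material for Niagara mesh particles, giving meshes a particle-colored Fresnel effect.
# Particle color RGB passes through the MF_FX_EmissiveAdjustments function for emissive adjustment.
# Particle alpha is multiplied by a Fresnel term with parameterized exponent and base reflect fraction.
# A static switch inverts the Fresnel term, swapping weighting between silhouette edges and camera-facing areas.
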Material /Game/Effects/Niagara/MasterMaterials/Mesh/M_SimpleMesh_Fresnel
graph {
 properties {
  bAutomaticallySetUsageInEditor=false
  bUsedWithNiagaraMeshParticles=true
  blendMode=BLEND_Translucent
  shadingModel=MSM_Unlit
  usedShadingModels=Unlit
 }
 nodes [
  n1 = ParticleColor()
  MaterialFunctionCall(Emissive=@n1.RGB) {materialFunction=&MF_FX_EmissiveAdjustments}
  n2 = Fresnel(
   ExponentIn=ScalarParameter() {defaultValue=5, parameterName=ExponentIn}
   BaseReflectFractionIn=ScalarParameter() {defaultValue=0.5, parameterName=BaseReflectFractionIn}
  ) {baseReflectFraction=0.5}
  Multiply(A=@n1.A, B=StaticSwitchParameter(True=OneMinus(None=@n2), False=@n2) {parameterName=Invert Fresnal})
 ]
}
