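# Library material function in the Niagara master materials that gives a locally aligned 0-1 UVW coordinate across each instance's local bounds.
# It transforms world position, excluding shader offsets, into instance space, subtracts the bounds minimum and divides by twice the extent.
# Custom HLSL nodes read the bounds center and extent directly from primitive data.
# Outputs are the combined RGB vector plus separate R, G and B components.
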MaterialFunction /Game/Effects/Niagara/MasterMaterials/Functions/InstanceBounds_0-1_UVW
graph {
 properties {bExposeToLibrary=true, description=Creates a locally aligned 0-1 value on XY and Z., libraryCategoriesText [Misc,Coordinates]}
 nodes [
  n1 = WorldPosition() {worldPositionShaderOffset=WPT_ExcludeAllShaderOffsets}
  n2 = Custom!(None) {code="GetPrimitiveData(Parameters).InstanceLocalBoundsCenter.xyz", description=BoundsCenter}
  n3 = Custom() {code="GetPrimitiveData(Parameters).InstanceLocalBoundsExtent.xyz", description=BoundsExtent}
  n4 = Divide(
   A=Subtract(
    A=TransformPosition(None=@n1.XYZ) {transformSourceType=TRANSFORMPOSSOURCE_World, transformType=TRANSFORMPOSSOURCE_Instance}
    B=Subtract(A=@n2, B=@n3)
   )
   B=Multiply(A=@n3) {constB=2}
  )
  FunctionOutput(None=@n4) {outputName=RGB}
  n5 = MaterialFunctionCall(Float3=@n4) {materialFunction=&BreakOutFloat3Components}
  FunctionOutput(None=@n5.R) {outputName=R, sortPriority=1}
  FunctionOutput(None=@n5.G) {outputName=G, sortPriority=2}
  FunctionOutput(None=@n5.B) {outputName=B, sortPriority=3}
 ]
}
