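# Niagara FX material function that colorizes material attributes with particle color and outputs an HDR emissive color.
# It reduces BaseColor to saturated vector-length grayscale, with an input to substitute particle color.
# "Correct Coloring" blends in particle-color multiplication before DeriveHDRfromLDR expands and tints the result, using a separate mobile range parameter.
# The time-of-day tint branch follows the output node's source and never reaches the output, despite the description.
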MaterialFunction /Game/Effects/Niagara/MasterMaterials/Functions/MF_FX_Colorize
graph {
 properties {
  bExposeToLibrary=true
  description="Colors the basecolor material attribute with particle color, curve and time of day."
  libraryCategoriesText [FX Base Color]
 }
 nodes [
  Constant()
  n1 = GetMaterialAttributes(None=FunctionInput() {inputType=FunctionInput_MaterialAttributes}) {attributeGetTypes [#1]}
  n2 = Saturate(None=MaterialFunctionCall(Vector 3=@n1.BaseColor) {materialFunction=/Engine/Functions/Engine_MaterialFunctions02/Utility/VectorLength}.V3 Length)
  n3 = ParticleColor()
  n4 = FunctionInput() {bUsePreviewValueAsDefault=true, inputName=Override LDR imput with particle color, inputType=FunctionInput_Scalar}
  n5 = GetMaterialAttributes(None=SetMaterialAttributes!(MaterialAttributes=@n1.MaterialAttributes, BaseColor=LinearInterpolate(A=@n2, B=@n3.RGB, Alpha=@n4)) {attributeSetTypes [#1]}) {attributeGetTypes [#1]}
  n6 = Constant() {r=1}
  n7 = MaterialFunctionCall(X=@n6, Y=@n6, Z=@n6) {materialFunction=&MakeFloat3}
  n8 = ParticleColor()
  n9 = LinearInterpolate(
   A=Saturate(None=Multiply(A=@n7.Result, B=@n5.BaseColor))
   B=Saturate(None=Multiply(A=@n5.BaseColor, B=@n8.RGB))
   Alpha=FunctionInput() {bUsePreviewValueAsDefault=true, inputName=Correct Coloring, sortPriority=1}
  )
  n10 = ParticleColor()
  n11 = Constant() {r=100}
  n12 = ScalarParameter() {defaultValue=1, parameterName="GLES HDR  Range"}
  n13 = SetMaterialAttributes!(
   MaterialAttributes=@n5.MaterialAttributes
   EmissiveColor=MaterialFunctionCall(
    LDR Input=LinearInterpolate(A=@n9, B=@n10.RGB, Alpha=@n4)
    HDR Dynamic Range=ShadingPathSwitch(Default=@n11, Deferred=@n11, Mobile=@n12) {desc=Migrated from FeatureLevelSwitch}
    HDR Tint=@n10.RGB
   ) {materialFunction=&DeriveHDRfromLDR}.RemappedRange
  ) {attributeSetTypes [#2]}
  FunctionOutput(None=@n13) {outputName="Color (MA)"}
  n14 = GetMaterialAttributes(None=@n13) {attributeGetTypes [#1]}
  n15 = VectorParameter() {group="3) Colorize Time of Day Tint", parameterName="Color_Morning (V3)", sortPriority=1, defaultValue=<rgba 1 1 1 1>}
  n16 = VectorParameter() {group="3) Colorize Time of Day Tint", parameterName="Color_Day (V3)", sortPriority=2, defaultValue=<rgba 1 1 1 1>}
  n17 = VectorParameter() {group="3) Colorize Time of Day Tint", parameterName="Color_Evening (V3)", sortPriority=3, defaultValue=<rgba 1 1 1 1>}
  n18 = VectorParameter() {group="3) Colorize Time of Day Tint", parameterName="Color_Night (V3)", sortPriority=4, defaultValue=<rgba 1 1 1 1>}
  n19 = Multiply(
   A=@n14.BaseColor
   B=MaterialFunctionCall(Color_Morning=@n15.RGB, Color_Day=@n16.RGB, Color_Evening=@n17.RGB, Color_Night=@n18.RGB) {materialFunction=&Color_Time_of_Day}.Result
  )
  SetMaterialAttributes!(MaterialAttributes=@n14.MaterialAttributes, BaseColor=@n19, EmissiveColor=@n19) {attributeSetTypes [#1, #2]}
 ]
}
