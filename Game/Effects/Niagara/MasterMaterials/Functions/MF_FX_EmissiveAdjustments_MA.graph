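# Material function for Niagara effect master materials that applies emissive adjustments to a Material Attributes input.
# It extracts the EmissiveColor attribute, masks it to RGB and passes it through MF_FX_EmissiveAdjustments.
# The adjusted result is written back into the original attributes, making it a Material Attributes wrapper around that function.
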MaterialFunction /Game/Effects/Niagara/MasterMaterials/Functions/MF_FX_EmissiveAdjustments_MA
graph {nodes [
  n1 = GetMaterialAttributes(None=FunctionInput() {inputType=FunctionInput_MaterialAttributes}) {attributeGetTypes [#1]}
  n2 = ComponentMask(None=@n1.EmissiveColor) {b=true, g=true, r=true}
  FunctionOutput(None=SetMaterialAttributes!(
    MaterialAttributes=@n1.MaterialAttributes
    EmissiveColor=MaterialFunctionCall(Emissive=@n2) {materialFunction=&MF_FX_EmissiveAdjustments}.Result
   ) {attributeSetTypes [#1]})
 ]}
