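# Material function, exposed to the library as "HSV to RGB" under Misc and VFX, that converts HSV to RGB to match Niagara.
# It breaks the HSV input into scalars and feeds hue to a Custom HLSL node that wraps it with fmod and builds a saturated RGB ramp from absolute-value math.
# That color is then offset from white, scaled by saturation, shifted back and multiplied by value.
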MaterialFunction /Game/Effects/Niagara/MasterMaterials/Functions/MF_FX_HSVtoRGB
graph {
 properties {bExposeToLibrary=true, description="HSV to RGB, matching Niagara", userExposedCaption=HSV to RGB, libraryCategoriesText [Misc,VFX]}
 nodes [
  MaterialExpressionFunctionInput_0 = FunctionInput() {inputName=HSV}
  n1 = Convert(None=@MaterialExpressionFunctionInput_0) {
   nodeName=Break Float3
   convertInputs [{
     type=Vector3
     expressionInput {
      expression=&self:MaterialExpressionFunctionInput_0
      inputName=None
      mask=0
      maskA=0
      maskB=0
      maskG=0
      maskR=0
      outputIndex=0
     }
    }]
   convertOutputs [{type=Scalar}, {type=Scalar}, {type=Scalar}]
  }
  MaterialExpressionReroute_2 = Reroute(None=@n1) {desc=Hue}
  n2 = Custom!(Hue=@MaterialExpressionReroute_2) {
   code="Hue = fmod(Hue, 1.0f);\r\nfloat R = abs(Hue * 6 - 3) - 1;\r\nfloat G = 2 - abs(Hue * 6 - 2);\r\nfloat B = 2 - abs(Hue * 6 - 4);\r\nfloat3 RGB = saturate(float3(R, G, B));\r\nreturn RGB;"
   description=RGB
   showCode=true
   inputs [{inputName=Hue, input {expression=&self:MaterialExpressionReroute_2}}]
  }
  n3 = Multiply(A=Subtract(A=@n2, B=Constant3Vector() {constant=<rgba 1 1 1 0>}), B=@n1)
  FunctionOutput(None=Multiply(A=Add(A=@n3), B=@n1))
 ]
}
