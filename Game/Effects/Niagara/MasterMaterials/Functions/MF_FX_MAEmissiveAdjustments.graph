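# Material function for FX materials, exposed to the library under FX Common, that applies last-in-line emissive adjustments to a Material Attributes input.
# It pulls EmissiveColor from the incoming attributes and passes it through MF_FX_EmissiveAdjustments.
# The result is written back with SetMaterialAttributes, leaving the other attributes untouched, so the emissive function works as a Material Attributes wrapper.
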MaterialFunction /Game/Effects/Niagara/MasterMaterials/Functions/MF_FX_MAEmissiveAdjustments
graph {
 properties {
  bExposeToLibrary=true
  description="Material Attribute container for Last-in-line Emissive Adjustments, etc."
  libraryCategoriesText [FX Common]
 }
 nodes [
  n1 = GetMaterialAttributes(None=FunctionInput() {inputType=FunctionInput_MaterialAttributes}) {attributeGetTypes [#1]}
  FunctionOutput(None=SetMaterialAttributes!(
    MaterialAttributes=@n1.MaterialAttributes
    EmissiveColor=MaterialFunctionCall(Emissive=@n1.EmissiveColor) {materialFunction=&MF_FX_EmissiveAdjustments}.Result
   ) {attributeSetTypes [#1]})
 ]
}
