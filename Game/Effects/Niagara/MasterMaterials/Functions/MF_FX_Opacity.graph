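# Material function for FX master materials that computes particle opacity and writes it to both Opacity and Opacity Mask of the incoming material attributes.
# A static switch, on by default, uses Base Color vector length; otherwise a channel-masked opacity texture is used, optionally blended toward its product with that length.
# Texture UVs come from a Niagara UV panner function plus a Distortion input.
# The result is saturated and multiplied by particle color alpha.
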MaterialFunction /Game/Effects/Niagara/MasterMaterials/Functions/MF_FX_Opacity
graph {nodes [
  n1 = GetMaterialAttributes(None=FunctionInput() {inputType=FunctionInput_MaterialAttributes}) {attributeGetTypes [#1, #2, #3]}
  n2 = MaterialFunctionCall(Vector 3=@n1.BaseColor) {materialFunction=/Engine/Functions/Engine_MaterialFunctions02/Utility/VectorLength}
  n3 = TextureCoordinate() {coordinateIndex=1}
  n4 = TextureCoordinate()
  n5 = ScalarParameter() {group="2.1) Opacity Uvs", parameterName="UV Channel (Opacity)", sliderMax=1, sortPriority=1}
  n6 = ScalarParameter() {defaultValue=1, group="2.1) Opacity Uvs", parameterName="U Freq (Opacity)", sortPriority=2}
  n7 = ScalarParameter() {defaultValue=1, group="2.1) Opacity Uvs", parameterName="V Freq (Opacity)", sortPriority=3}
  n8 = ScalarParameter() {defaultValue=1, group="2.1) Opacity Uvs", parameterName="Pan U Speed (Opacity)", sortPriority=4}
  n9 = ScalarParameter() {group="2.1) Opacity Uvs", parameterName="Pan V Speed (Opacity)", sortPriority=5}
  n10 = TextureSampleParameter2D(UVs=Add(
    A=FunctionInput() {bUsePreviewValueAsDefault=true, inputName=Distortion, inputType=FunctionInput_Vector2}
    B=MaterialFunctionCall!(
     UV Channel 2=@n3
     UV Channel 1=@n4
     In=@n1.MaterialAttributes
     UV Channel=@n5
     U Freq Multiplier=@n6
     V Freq Multiplier=@n7
     Pan Speed U=@n8
     Pan Speed V=@n9
    ) {desc=The Niagara Bridge, materialFunction=/Game/Effects/Niagara/MasterMaterials/Functions/UVs/MF_NFX_TextureUVPanner}.Result
   )) {group="2) Opacity", parameterName=Opacity Texture, sortPriority=1, texture=/Game/Global/Textures/Default/Masks_and_Gradients/T_SparkRound}
  n11 = ChannelMaskParameter(None=@n10.RGB) {group="2) Opacity", parameterName=Opacity Texture Mask, sortPriority=2}
  n12 = StaticSwitchParameter(
   True=@n2.V3 Length
   False=LinearInterpolate(
    A=@n11.R
    B=Multiply(A=@n2.V3 Length, B=@n11.R)
    Alpha=ScalarParameter() {group="2) Opacity", parameterName="Multipy Opacity Texture Against Base Color)", sliderMax=1}
   )
  ) {
   defaultValue=true
   desc="If Activated, you can use a seperate texture read for your opacity. Otherwise, it inherits from the basecolor material attribute."
   group="2) Opacity"
   parameterName=Use Base Color for Opacity
  }
  n13 = Multiply(A=Saturate(None=@n12), B=ParticleColor().A)
  FunctionOutput(None=SetMaterialAttributes!(MaterialAttributes=@n1.MaterialAttributes, Opacity=@n13, OpacityMask=@n13) {attributeSetTypes [#1, #2]}) {outputName=Opacity}
  StaticSwitch(True=@n11.R)
 ]}
