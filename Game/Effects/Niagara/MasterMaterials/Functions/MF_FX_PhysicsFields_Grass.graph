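# Material function that adds grass world position offset from the global physics field's linear velocity.
# A time-panned, Z-stretched curl noise texture distorts the field sample position and perturbs its direction.
# That direction, optionally flattened and projected against vertex normals, scales by field speed, strength and a UV1-based height mask.
# A precision function cleans up the sum; mobile and non-Epic quality pass the previous WPO through.
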MaterialFunction /Game/Effects/Niagara/MasterMaterials/Functions/MF_FX_PhysicsFields_Grass
graph {
 properties {
  bExposeToLibrary=true
  description=Samples the global physics field linear velocity and returns a float3. Only works on EPIC.
  libraryCategoriesText [Physics Fields]
 }
 nodes [
  Constant() {r=50}
  n1 = FunctionInput(Preview=Constant3Vector()) {bUsePreviewValueAsDefault=true, desc=Previous WPO value that this function should add to., inputName=Previous WPO Value}
  n2 = ScalarParameter() {
   defaultValue=1
   desc=This scales perlin noise to add into the physics field linear velocity sample.
   group=Physics Field
   parameterName=Direction Noise Scale
  }
  n3 = Constant() {r=600}
  n4 = MaterialFunctionCall(Float3=Divide(
    A=WorldPosition().XYZ
    B=FunctionInput(Preview=@n3) {bUsePreviewValueAsDefault=true, inputName=Noise Size, inputType=FunctionInput_Scalar, sortPriority=3}
   )) {materialFunction=&BreakOutFloat3Components}
  n5 = FunctionInput(Preview=Constant() {r=8}) {
   bUsePreviewValueAsDefault=true
   desc=divides absolute world Z
   inputName=Stretch Noise Coordinates Z
   inputType=FunctionInput_Scalar
   sortPriority=6
  }
  n6 = Divide(A=@n4.B, B=@n5) {constB=8, desc=stretch in one direction}
  n7 = FunctionInput(Preview=Constant() {desc=Default, r=0.25}) {
   bUsePreviewValueAsDefault=true
   desc=Multiplies time
   description=This value pans a curl noise volume texture in z
   inputName=Noise Pan Speed
   inputType=FunctionInput_Scalar
   sortPriority=4
  }
  n8 = Multiply(A=@n7, B=Time())
  n9 = TextureSample(UVs=MaterialFunctionCall(X=@n4.R, Y=@n4.R, Z=Add(A=@n6, B=@n8)) {materialFunction=&MakeFloat3}.Result) {samplerType=SAMPLERTYPE_LinearColor, texture=/Game/Effects/FogVolume/VT_CurlNoise_LOW}
  n10 = Multiply(A=@n9.RGB, B=@n2)
  n11 = ScalarParameter() {group=Physics Field, parameterName=UV Distortion Strength}
  n12 = SamplePhysicsVectorField!(World Position=Add(A=Multiply(A=@n9.RGB, B=@n11), B=WorldPosition().XYZ)) {fieldTarget=Vector_LinearVelocity}
  n13 = MaterialFunctionCall(Vector 3=Add(A=@n12) {constB=0.0001}) {materialFunction=/Engine/Functions/Engine_MaterialFunctions02/Utility/VectorLength}
  n14 = FunctionInput(Preview=ScalarParameter() {defaultValue=4, group=Physics Field, parameterName=Noise Falloff}) {bUsePreviewValueAsDefault=true, inputName=Noise Falloff, inputType=FunctionInput_Scalar, sortPriority=5, previewValue=<2 0 0 1>}
  n15 = Power(Base=@n13.V3 Length, Exp=@n14) {constExponent=5}
  n16 = Normalize(VectorInput=Add(A=Multiply(A=@n10, B=@n15) {desc=Increase nosie falloff}, B=@n12))
  n17 = VertexNormalWS()
  n18 = Normalize(VectorInput=MaterialFunctionCall!(Plane Normal=Constant3Vector() {constant=<rgba 0 0 1 0>}, Vector=@n16) {materialFunction=/Engine/Functions/Engine_MaterialFunctions02/Math/ProjectVectorOntoPlane}.Result)
  n19 = FunctionInput(Preview=Constant() {r=1}) {bUsePreviewValueAsDefault=true, inputName=Project Field Force Vector to Normals, inputType=FunctionInput_Scalar, sortPriority=2}
  n20 = LinearInterpolate(A=@n16, B=CrossProduct(A=CrossProduct(A=@n17, B=@n18), B=@n17), Alpha=@n19) {desc=Option to project vector}
  n21 = ScalarParameter() {defaultValue=50, parameterName=Physic Field Strength}
  n22 = MaterialFunctionCall(Float2=TextureCoordinate() {coordinateIndex=1}) {materialFunction=&BreakOutFloat2Components}
  n23 = FunctionInput(Preview=OneMinus(None=Multiply(A=@n22.G, B=@n22.G))) {
   bUsePreviewValueAsDefault=true
   description=Mask used to determine the vertical falloff of the force magnitude amount
   inputName=Force Magnitude Mask
   inputType=FunctionInput_Scalar
   sortPriority=1
  }
  n24 = Multiply(
   A=@n13.V3 Length
   B=Multiply(
    A=FunctionInput(Preview=@n21) {bUsePreviewValueAsDefault=true, inputName=Force Magnitude Multiplier, inputType=FunctionInput_Scalar, sortPriority=1}
    B=@n23
   )
  )
  n25 = MaterialFunctionCall!(InVector=Add(A=@n1, B=Multiply(A=@n20, B=@n24) {desc=Multiply strength of the physics field vector after projecting it})) {
   desc=essential to remove artifacting from differences in the depth buffer
   materialFunction=/Game/Athena/Environments/Landscape/Foliage/MF_MakePreciseVec3
  }
  FunctionOutput(None=ShadingPathSwitch(Default=QualitySwitch(Default=@n1, Epic=@n25.Result), Mobile=@n1))
 ]
}
