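# Material function for FX materials that tiles and pans a texture from packed vector inputs, exposed to the library.
# InRGB red and green scale the UVs, while blue and the separate InA scalar set X and Y pan speed, with time scaled by TimeMult.
# ClampTexture blends panned UVs toward their saturated value.
# It outputs both tiled and panned UVs, plus RGB samples of the input texture at each.
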MaterialFunction /Game/Effects/Niagara/MasterMaterials/Functions/MF_FX_VectorDrive_TexPanner
graph {
 properties {bExposeToLibrary=true}
 nodes [
  n1 = TextureCoordinate()
  n2 = FunctionInput(Preview=Constant3Vector() {constant=<rgba 1 1 1 1>}) {bUsePreviewValueAsDefault=true, inputName=InRGB, sortPriority=3}
  n3 = ComponentMask(None=@n2) {r=true}
  n4 = ComponentMask(None=@n2) {g=true}
  n5 = FunctionOutput(None=Multiply(
    A=FunctionInput(Preview=@n1) {bUsePreviewValueAsDefault=true, inputName=UV, inputType=FunctionInput_Vector2, sortPriority=1}
    B=AppendVector(A=@n3, B=@n4)
   )) {outputName=TiledUV, sortPriority=3}
  n6 = Constant() {r=1}
  n7 = Time()
  n8 = Panner(
   Coordinate=@n5
   Time=Multiply(
    A=FunctionInput(Preview=@n6) {bUsePreviewValueAsDefault=true, inputName=TimeMult, inputType=FunctionInput_Scalar, sortPriority=6}
    B=FunctionInput(Preview=@n7) {bUsePreviewValueAsDefault=true, inputName=Time, inputType=FunctionInput_Scalar, sortPriority=5}
   )
   Speed=AppendVector(
    A=ComponentMask(None=@n2) {b=true}
    B=FunctionInput() {bUsePreviewValueAsDefault=true, inputName=InA, inputType=FunctionInput_Scalar, sortPriority=4}
   )
  )
  n9 = LinearInterpolate(A=@n8, B=Saturate(None=@n8), Alpha=FunctionInput() {bUsePreviewValueAsDefault=true, inputName=ClampTexture, inputType=FunctionInput_Scalar})
  n10 = FunctionInput(Preview=TextureObject() {texture=/Game/Global/Textures/Default/Masks_and_Gradients/3pxRGBImage}) {bUsePreviewValueAsDefault=true, inputName=Tex, inputType=FunctionInput_Texture2D, sortPriority=2}
  FunctionOutput(None=TextureSample(UVs=FunctionOutput(None=@n9) {outputName=PannerUV, sortPriority=1}, Tex=@n10).RGB) {outputName=PannerTexture, sortPriority=2}
  FunctionOutput(None=TextureSample(UVs=@n5, Tex=@n10).RGB) {outputName=TiledTexture, sortPriority=4}
 ]
}
