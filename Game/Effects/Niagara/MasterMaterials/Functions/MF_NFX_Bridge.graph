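# MF_NFX_Bridge is a material function that passes per-particle Niagara dynamic parameters to effect master materials as named outputs.
# It reads all four dynamic parameter slots: texture and distortion UV time with strengths, UV time with Add, PostMult and Pow, RGBA component masks, and UV0/UV1 frequencies.
# Each channel goes straight to an output without processing, and the mask slot pairs with the Niagara Material ComponentMask module.
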MaterialFunction /Game/Effects/Niagara/MasterMaterials/Functions/MF_NFX_Bridge
graph {nodes [
  n1 = DynamicParameter() {paramNames [Base Texture UV Time, Distortion UV Time, Distortion U Strength, Distortion V Strength]}
  FunctionOutput(None=@n1.Base Texture UV Time) {description=Base Texture UV Time, outputName=Base Texture UV Time}
  FunctionOutput(None=@n1.Distortion UV Time) {description=Distortion Texture UV Time, outputName=Distortion UV Time, sortPriority=1}
  n2 = DynamicParameter() {
   desc=This talks to the niagara module Material ComponentMask
   parameterIndex=2
   paramNames [R Component Mask, G Component Mask, B Component Mask, A Component Mask]
  }
  FunctionOutput(None=@n2.B Component Mask) {description=B Component Mask, outputName=B Component Mask, sortPriority=7}
  FunctionOutput(None=@n2.A Component Mask) {description=A Component Mask, outputName=A Component Mask, sortPriority=7}
  n3 = DynamicParameter() {parameterIndex=3, paramNames [UV0 U Freq, UV0 V Freq, UV1 U Freq, UV1 V Freq]}
  FunctionOutput(None=@n3.UV0 U Freq) {description=UV0 U Freq, outputName=UV0 U Freq, sortPriority=7}
  FunctionOutput(None=@n3.UV0 V Freq) {description=UV0 V Freq, outputName=UV0 V Freq, sortPriority=7}
  FunctionOutput(None=@n3.UV1 U Freq) {description=UV1 U Freq, outputName=UV1 U Freq, sortPriority=7}
  FunctionOutput(None=@n3.UV1 V Freq) {description=UV1 V Freq, outputName=UV1 V Freq, sortPriority=7}
  FunctionOutput(None=@n1.Distortion U Strength) {description=Distortion Strength, outputName=Distortion U Strength, sortPriority=2}
  FunctionOutput(None=@n1.Distortion V Strength) {description=Distortion Strength, outputName=Distortion V Strength, sortPriority=3}
  n4 = DynamicParameter() {parameterIndex=1, defaultValue=<rgba 0 0 1 1>, paramNames [UV Time, Add, PostMult, Pow]}
  FunctionOutput(None=@n4.UV Time) {description=UV Time, outputName=UV Time, sortPriority=4}
  FunctionOutput(None=@n4.Pow) {description=Pow, outputName=Pow, sortPriority=7}
  FunctionOutput(None=@n4.PostMult) {description=PostMult, outputName=PostMult, sortPriority=6}
  FunctionOutput(None=@n4.Add) {description=Add, outputName=Add, sortPriority=5}
  FunctionOutput(None=@n2.R Component Mask) {description=R Component Mask, outputName=R Component Mask, sortPriority=7}
  FunctionOutput(None=@n2.G Component Mask) {description=G Component Mask, outputName=G Component Mask, sortPriority=7}
 ]}
