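# MF_NFX_ComponentMask is a library material function for Niagara effects that masks a previously set BaseColor using the Niagara component mask module.
# It reads the R Component Mask from MF_NFX_Bridge, replicates it across three channels and multiplies it with BaseColor.
# The product is split and collapsed to a scalar as red plus red plus blue, leaving green unused.
# That scalar overwrites BaseColor in the returned material attributes.
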MaterialFunction /Game/Effects/Niagara/MasterMaterials/Functions/MF_NFX_ComponentMask
graph {
 properties {bExposeToLibrary=true, description=Uses the Niagara Material COmponent Mask Moudle, libraryCategoriesText [FX BaseColor]}
 nodes [
  n1 = GetMaterialAttributes(None=FunctionInput() {description=BaseColor should be previously set by a MF_FX_BaseColor Material Function., inputType=FunctionInput_MaterialAttributes}) {attributeGetTypes [#1]}
  n2 = MaterialFunctionCall() {materialFunction=&/MF_NFX_Bridge}
  n3 = MaterialFunctionCall(Color=Multiply(
    A=@n1.BaseColor
    B=MaterialFunctionCall(X=@n2.R Component Mask, Y=@n2.R Component Mask, Z=@n2.R Component Mask) {materialFunction=&MakeFloat3}.Result
   )) {materialFunction=&SplitComponents}
  n4 = Add(A=@n3.R, B=@n3.R)
  FunctionOutput(None=SetMaterialAttributes!(MaterialAttributes=@n1.MaterialAttributes, BaseColor=Add(A=@n4, B=@n3.B)) {attributeSetTypes [#1]}) {outputName="BaseColor (MA) "}
 ]
}
