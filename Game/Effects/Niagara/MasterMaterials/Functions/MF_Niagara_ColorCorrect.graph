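# Library material function that color corrects base color in Niagara effect materials, expecting it to be set earlier by MF_FX_BaseColor.
# It reads BaseColor from the incoming material attributes and Add, PostMult and Pow values from MF_NFX_Bridge.
# The color is offset, multiplied, raised to the power, then floored at zero.
# The result overwrites both BaseColor and EmissiveColor in the output attributes.
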MaterialFunction /Game/Effects/Niagara/MasterMaterials/Functions/MF_Niagara_ColorCorrect
graph {
 properties {
  bExposeToLibrary=true
  description=This uses the NIagara bridge to color correct the base color material attribute.
  libraryCategoriesText [FX BaseColor]
 }
 nodes [
  n1 = GetMaterialAttributes(None=FunctionInput() {description=Basecolor should be set previously by a MF_FX_BaseColor., inputType=FunctionInput_MaterialAttributes}) {attributeGetTypes [#1]}
  n2 = MaterialFunctionCall() {materialFunction=&/MF_NFX_Bridge}
  n3 = Add(A=@n1.BaseColor, B=@n2.Add)
  n4 = Max(A=Power(Base=Multiply(A=@n3, B=@n2.PostMult), Exp=@n2.Pow) {constExponent=0.5}) {constB=0}
  FunctionOutput(None=SetMaterialAttributes!(MaterialAttributes=@n1.MaterialAttributes, BaseColor=@n4, EmissiveColor=@n4) {attributeSetTypes [#1, #2]}) {outputName="BaseColor (MA)"}
 ]
}
