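# MF_Niagara_Erosion is a material function that exposes a per-particle erosion value to Niagara effect materials.
# It reads the Erosion channel, the first component of Dynamic Parameter index 2, and passes it straight to its Erosion output with no further math.
# The MaterialErosion Niagara module drives the value, which defaults to zero.
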MaterialFunction /Game/Effects/Niagara/MasterMaterials/Functions/MF_Niagara_Erosion
graph {nodes [FunctionOutput(None=DynamicParameter() {parameterIndex=2, defaultValue=<rgba 0 0 0 0>, paramNames [Erosion,Param2,Param3,Param4]}.Erosion) {description=This value is driven by the MaterialErosion Niagara Module., outputName=Erosion}]}
