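# Material function, exposed to the library under Niagara, that gives materials the values of a Niagara Material Panner module.
# It reads Dynamic Parameter index 3, matching Particles.MaterialDynamicParameter3, with channels named Amplitude, Frequency, Offset and Axis.
# Each channel passes straight to its own output, with Axis exposed as Vector Component; no panning math happens inside the function.
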MaterialFunction /Game/Effects/Niagara/MasterMaterials/Functions/MF_Niagara_MaterialPannerParameters
graph {
 properties {
  bExposeToLibrary=true
  description="Reads from Niagara Module Material Panner, Particles.MaterialDynamicParameter3"
  libraryCategoriesText [Niagara]
 }
 nodes [
  n1 = DynamicParameter() {parameterIndex=3, paramNames [Amplitude,Frequency,Offset,Axis]}
  FunctionOutput(None=@n1.Frequency) {description="Reads from Niagara Module Material Panner, Particles.MaterialDynamicParameter3.y", outputName=Frequency, sortPriority=1}
  FunctionOutput(None=@n1.Amplitude) {description="Reads from Niagara Module Material Panner, Particles.MaterialDynamicParameter3.x", outputName=Amplitude}
  FunctionOutput(None=@n1.Offset) {description="Reads from Niagara Module Material Panner, Particles.MaterialDynamicParameter3.z", outputName=Offset, sortPriority=2}
  FunctionOutput(None=@n1.Axis) {description="Reads from Niagara Module Material Panner, Particles.MaterialDynamicParameter3.a", outputName=Vector Component, sortPriority=4}
 ]
}
