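# Material function for Niagara effect materials that scales texture coordinates and offsets them along U.
# MF_Niagara_MaterialPannerParameters supplies Frequency and Offset, which combine with function inputs for extra scaling and shifting.
# U is multiplied by both frequencies, then shifted by the multiplied parameter offset plus the input Offset.
# V is scaled only by its own frequency input, and both channels are appended into the output UV.
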MaterialFunction /Game/Effects/Niagara/MasterMaterials/Functions/MF_Niagara_Panner
graph {nodes [
  n1 = FunctionInput() {description=Input regular texture coordinates, inputName=Texture Coordinates, inputType=FunctionInput_Vector2}
  n2 = MaterialFunctionCall() {materialFunction=&/MF_Niagara_MaterialPannerParameters}
  n3 = FunctionInput() {
   bUsePreviewValueAsDefault=true
   description=Scale U
   inputName=U Frequency
   inputType=FunctionInput_Scalar
   sortPriority=1
   previewValue=<1 0 0 1>
  }
  n4 = Multiply(A=ComponentMask(None=@n1) {r=true}, B=Multiply(A=@n2.Frequency, B=@n3))
  n5 = FunctionInput() {bUsePreviewValueAsDefault=true, inputName=OffsetMultiplier, inputType=FunctionInput_Scalar, sortPriority=10, previewValue=<1 0 0 0>}
  n6 = FunctionInput() {
   bUsePreviewValueAsDefault=true
   description=Offset U
   inputName=Offset
   inputType=FunctionInput_Scalar
   sortPriority=3
   previewValue=<1 0 0 1>
  }
  n7 = Add(A=Multiply(A=@n2.Offset, B=@n5), B=@n6)
  n8 = ComponentMask(None=@n1) {g=true}
  n9 = FunctionInput() {
   bUsePreviewValueAsDefault=true
   description=Scale V
   inputName=V Frequency
   inputType=FunctionInput_Scalar
   sortPriority=2
   previewValue=<1 0 0 1>
  }
  FunctionOutput(None=AppendVector(A=Add(A=@n4, B=@n7), B=Multiply(A=@n8, B=@n9)))
 ]}
